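# Physics asset for the collar mesh of the male medium CrispRover Breeze body cosmetic, setting up physics for its tie bones.
# Two three-box chains, dyn_tie and dyn_tie_down, branch from a base box with limited swing and twist at each joint.
# The base is rotationally locked to spine_04, which chains through spine_02 to the pelvis.
# Capsules approximate pelvis, spine and thighs, with the thigh bodies left unconstrained.
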PhysicsAsset /BRCosmetics/Characters/Player/Male/Medium/Bodies/M_MED_CrispRover_Breeze/Meshes/M_MED_CrispRover_Breeze_Collar_Physics
bodies {
 dyn_tie_1 [{shapeType=Box, extentX=8.5, extentY=1.2823, extentZ=6.5, shapeName=dyn_tie_1_box, center=<4.2885 0.6214 0>, rotation=<rot 0 -0.1601 0>}]
 dyn_tie_2 [{shapeType=Box, extentX=8.5, extentY=1.5381, extentZ=7.5, shapeName=dyn_tie_2_box, center=<4.2874 0.233 0>, rotation=<rot 0 -2.1419 0>}]
 dyn_tie_3 [{shapeType=Box, extentX=11.8177, extentY=9.5, extentZ=1.4189, shapeName=dyn_tie_3_box, center=<6.5848 0 0.191>}]
 dyn_tie_base [{shapeType=Box, extentX=9.5, extentY=2.5, extentZ=4.5, shapeName=dyn_tie_base_box, center=<5.0914 0 0>}]
 dyn_tie_down_1 [{shapeType=Box, extentX=7.2622, extentY=1.4864, extentZ=3.5, shapeName=dyn_tie_down_1_box, center=<3.8637 0.1501 0>}]
 dyn_tie_down_2 [{shapeType=Box, extentX=8.5, extentY=1.3414, extentZ=6.5, shapeName=dyn_tie_down_2_box, center=<3.9366 0 0>}]
 dyn_tie_down_3 [{shapeType=Box, extentX=12.5, extentY=9, extentZ=1.4791, shapeName=dyn_tie_down_3_box, center=<7.6884 0 0>}]
 pelvis [{shapeType=Capsule, length=8.8375, radius=10.6297, center=<0.8525 0.1953 0.4506>, rotation=<rot 0 -81.5525 -1.5748>}]
 spine_02 [{shapeType=Capsule, length=10.3663, radius=14.5001, center=<3.8355 -3.5153 -0.854>, rotation=<rot -9.2225 -84.265 -91.2214>}]
 spine_04 [{shapeType=Capsule, length=9.9877, radius=13.1915, center=<3.3798 -1.8494 0.0723>, rotation=<rot -0.736 -96.5024 -90.5623>}]
 thigh_l [{
   shapeType=Capsule
   length=22.0952
   radius=7.9048
   shapeName=thigh_l_capsule
   center=<-17.294 0.879 -0.0844>
   rotation=<rot -89.2331 -0.0044 2.5489>
  }]
 thigh_r [{
   shapeType=Capsule
   length=22.0952
   radius=7.9048
   shapeName=thigh_r_capsule
   center=<16.9518 -0.879 -0.77>
   rotation=<rot -88.9786 -0.006 2.5505>
  }]
}
constraints [
 {bone1Name=spine_02, bone2Name=pelvis, swing1LimitDegrees=45, swing2LimitDegrees=45, twistLimitDegrees=45}
 {bone1Name=spine_04, bone2Name=spine_02, swing1LimitDegrees=45, swing2LimitDegrees=45, twistLimitDegrees=45}
 {
  bone1Name=dyn_tie_base
  bone2Name=spine_04
  swing1LimitDegrees=45
  swing1Motion=Locked
  swing2LimitDegrees=45
  swing2Motion=Locked
  twistLimitDegrees=45
  twistMotion=Locked
 }
 {bone1Name=dyn_tie_1, bone2Name=dyn_tie_base, swing1LimitDegrees=20, swing2LimitDegrees=10, twistLimitDegrees=10}
 {bone1Name=dyn_tie_down_1, bone2Name=dyn_tie_base, swing1LimitDegrees=15, swing2LimitDegrees=10, twistLimitDegrees=10}
 {bone1Name=dyn_tie_2, bone2Name=dyn_tie_1, swing1LimitDegrees=20, swing2LimitDegrees=10, twistLimitDegrees=10}
 {bone1Name=dyn_tie_3, bone2Name=dyn_tie_2, swing1LimitDegrees=10, swing2LimitDegrees=20, twistLimitDegrees=10}
 {bone1Name=dyn_tie_down_2, bone2Name=dyn_tie_down_1, swing1LimitDegrees=15, swing2LimitDegrees=10, twistLimitDegrees=10}
 {bone1Name=dyn_tie_down_3, bone2Name=dyn_tie_down_2, swing1LimitDegrees=15, swing2LimitDegrees=20, twistLimitDegrees=15}
]
