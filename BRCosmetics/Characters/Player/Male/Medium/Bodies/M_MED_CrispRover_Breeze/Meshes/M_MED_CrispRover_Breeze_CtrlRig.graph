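# Control Rig for the CrispRover Breeze player body, procedurally posing its coat, collar, strap, tie, kneepad, keychain and dangling chains.
# Spherical pose readers on thighs, upper arms and head weight corrective rotations, adjusted by crouch, fly, roll and frontend variables.
# Four front/back chains use stretchy B-spline IK on controls constrained to body bones.
# Construction caches offset control poses, which thigh-raise alphas blend against bind poses; the right-thigh back-out branch is unconnected.
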ControlRigBlueprint /BRCosmetics/Characters/Player/Male/Medium/Bodies/M_MED_CrispRover_Breeze/Meshes/M_MED_CrispRover_Breeze_CtrlRig
graphs [
 ContainedGraph {nodes [
   n1 = RigVMFunctionEntryNode()
   SequenceAggregate = &RigVMFunction_Sequence(ExecuteContext=@n1.ExecuteContext)
   SequenceAggregate_1 = &RigVMFunction_Sequence(ExecuteContext=@SequenceAggregate.B)
   SequenceAggregate_2 = &RigVMFunction_Sequence(ExecuteContext=@SequenceAggregate_1.B)
   SequenceAggregate_3 = &RigVMFunction_Sequence(ExecuteContext=@SequenceAggregate_2.B)
   SequenceAggregate_4 = &RigVMFunction_Sequence(ExecuteContext=@SequenceAggregate_3.B)
   SequenceAggregate_5 = &RigVMFunction_Sequence(ExecuteContext=@SequenceAggregate_4.B)
   RigVMFunctionReturnNode(
    A=@SequenceAggregate.A
    B=@SequenceAggregate_1.A
    C=@SequenceAggregate_2.A
    D=@SequenceAggregate_3.A
    E=@SequenceAggregate_4.A
    F=@SequenceAggregate_5.A
    G=@SequenceAggregate_5.B
   )
  ]}
 ContainedGraph {nodes [
   n1 = RigVMFunctionEntryNode()
   SequenceAggregate = &RigVMFunction_Sequence(ExecuteContext=@n1.ExecuteContext)
   SequenceAggregate_1 = &RigVMFunction_Sequence(ExecuteContext=@SequenceAggregate.B)
   RigVMFunctionReturnNode(A=@SequenceAggregate.A, B=@SequenceAggregate_1.A, C=@SequenceAggregate_1.B)
  ]}
 RigVMFunctionLibrary {links [], nodes []}
 RigVMModel {nodes [
   n1 = &RigVMFunction_ControlFlowBranch(ExecuteContext=())
   n2 = RigVMVariableNode() {variable=SplineIK_chain_r_ctrls}
   /Script/ControlRig.RigUnit_HierarchySetPoseItemArray(
    ElementType=All
    ExecutePin=@n1.True
    Weight=&RigVMFunction_MathDoubleSub(A=1, B=RigVMVariableNode() {variable=spr_thigh_r_slight_back_out_apha}.Value).Result
    Pose=RigVMVariableNode() {variable=SplineIK_ctrls_r_bind_pose}.Value
    ItemsToSet=@n2.Value
   ) {is=struct RigVMUnitNode}
   > &RigUnit_HierarchySetPoseItemArray(ElementType=All, Weight=1, Pose=RigVMVariableNode() {variable=SplineIK_ctrl_thigh_r_leg_slight_back_out_pose}.Value, ItemsToSet=@n2.Value)
   RigUnit_BeginExecution = &RigUnit_BeginExecution()
   SequenceAggregate = RigVMAggregateNode(ExecuteContext=@RigUnit_BeginExecution.ExecutePin)
   &RigUnit_ParentConstraint(
    AdvancedSettings=(RotationOrderForFilter=XZY)
    Child=(Type=Null,Name="dyn_chain_ctrl_l_null")
    Parents=((Item=(Type=Bone,Name="dyn_chain_ctrl_end_l"),Weight=1))
    Weight=1
    ExecutePin=@SequenceAggregate.A
   )
   > &RigUnit_ParentConstraint(
    AdvancedSettings=(RotationOrderForFilter=XZY)
    Child=(Type=Null,Name="dyn_chain_ctrl_r_null")
    Parents=((Item=(Type=Bone,Name="dyn_chain_ctrl_end_r"),Weight=1))
    Weight=1
   )
   n3 = &RigUnit_SphericalPoseReader(
    ActiveRegionScaleFactors=(PositiveWidth=3.5,NegativeWidth=0.3,PositiveHeight=4,NegativeHeight=3)
    ActiveRegionSize=0.1
    Debug=(bDrawDebug=false,DebugScale=25,DebugSegments=20,DebugThickness=0.25)
    DriverAxis=(X=-1,Y=0,Z=0)
    DriverItem=(Type=Bone,Name="thigh_l")
    FalloffRegionScaleFactors=(PositiveWidth=1,NegativeWidth=1.3,PositiveHeight=2,NegativeHeight=2)
    FalloffSize=0.45
    OptionalParentItem=(Type=Bone)
    RotationOffset=(X=90,Y=-20,Z=-15)
    ExecutePin=@SequenceAggregate.B
   )
   > RigVMVariableNode(Value=@n3.OutputParam) {variable=spr_thigh_l_up_alpha}
   > n4 = &RigUnit_SphericalPoseReader(
    ActiveRegionScaleFactors=(PositiveWidth=3.5,NegativeWidth=0.3,PositiveHeight=4,NegativeHeight=3)
    ActiveRegionSize=0.1
    Debug=(bDrawDebug=false,DebugScale=25,DebugSegments=20,DebugThickness=0.25)
    DriverAxis=(X=1,Y=0,Z=0)
    DriverItem=(Type=Bone,Name="thigh_r")
    FalloffRegionScaleFactors=(PositiveWidth=1,NegativeWidth=1.3,PositiveHeight=2,NegativeHeight=2)
    FalloffSize=0.45
    FlipHeightScaling=false
    FlipWidthScaling=true
    OptionalParentItem=(Type=Bone)
    RotationOffset=(X=-90,Y=-20,Z=-15)
   )
   > RigVMVariableNode(Value=@n4.OutputParam) {variable=spr_thigh_r_up_alpha}
   &RigUnit_ParentConstraint(
    AdvancedSettings=(RotationOrderForFilter=XZY)
    Child=(Type=Null,Name="dyn_chain_base_front_l_null")
    Parents=((Item=(Type=Bone,Name="pelvis"),Weight=0.62),(Item=(Type=Bone,Name="deform_groin_l"),Weight=0.19),(Item=(Type=Bone,Name="thigh_twist_01_l"),Weight=0.15),(Item=(Type=Bone,Name="spine_01"),Weight=0.04))
    Weight=1
    ExecutePin=@SequenceAggregate.C
   )
   > &RigUnit_ParentConstraint(
    AdvancedSettings=(RotationOrderForFilter=XZY)
    Child=(Type=Null,Name="dyn_chain_base_back_l_null")
    Parents=((Item=(Type=Bone,Name="pelvis"),Weight=0.63),(Item=(Type=Bone,Name="deform_glute_l"),Weight=0.28),(Item=(Type=Bone,Name="thigh_twist_01_l"),Weight=0.09))
    Weight=1
   )
   > &RigUnit_ParentConstraint(
    AdvancedSettings=(RotationOrderForFilter=XZY)
    Child=(Type=Null,Name="dyn_chain_base_front_r_null")
    Parents=((Item=(Type=Bone,Name="pelvis"),Weight=0.62),(Item=(Type=Bone,Name="deform_groin_r"),Weight=0.19),(Item=(Type=Bone,Name="thigh_twist_01_r"),Weight=0.15),(Item=(Type=Bone,Name="spine_01"),Weight=0.04))
    Weight=1
   )
   > &RigUnit_ParentConstraint(
    AdvancedSettings=(RotationOrderForFilter=XZY)
    Child=(Type=Null,Name="dyn_chain_base_back_r_null")
    Parents=((Item=(Type=Bone,Name="pelvis"),Weight=0.77),(Item=(Type=Bone,Name="deform_glute_r"),Weight=0.17),(Item=(Type=Bone,Name="thigh_twist_01_r"),Weight=0.06))
    Weight=1
   )
   SequenceAggregate_1 = &RigVMFunction_Sequence(ExecuteContext=@SequenceAggregate.D)
   &RigUnit_ParentConstraint(
    AdvancedSettings=(RotationOrderForFilter=XZY)
    Child=(Type=Null,Name="dyn_chain_3_front_l_null")
    Parents=((Item=(Type=Bone,Name="thigh_l"),Weight=0.75),(Item=(Type=Control,Name="dyn_chain_ctrl_l_control"),Weight=0.25))
    Weight=1
    ExecutePin=@SequenceAggregate_1.A
   )
   > &RigUnit_ParentConstraint(
    AdvancedSettings=(RotationOrderForFilter=XZY)
    Child=(Type=Null,Name="dyn_chain_6_front_l_null")
    Parents=((Item=(Type=Bone,Name="thigh_l"),Weight=0.5),(Item=(Type=Control,Name="dyn_chain_ctrl_l_control"),Weight=0.5))
    Weight=1
   )
   n5 = RigVMVariableNode() {variable=spr_thigh_l_up_alpha}
   n6 = &RigVMFunction_MathDoubleSub(A=1, B=@n5.Value)
   > &RigUnit_ParentConstraint(
    AdvancedSettings=(RotationOrderForFilter=XZY)
    Child=(Type=Null,Name="dyn_chain_2_back_l_null")
    Parents=((Item=(Type=Bone,Name="thigh_l"),Weight=0.75),(Item=(Type=Control,Name="dyn_chain_ctrl_l_control"),Weight=0.25))
    Weight=@n6.Result
   )
   > &RigUnit_ParentConstraint(
    AdvancedSettings=(RotationOrderForFilter=XZY)
    Child=(Type=Null,Name="dyn_chain_back_3_l_null")
    Parents=((Item=(Type=Bone,Name="thigh_l"),Weight=0.25),(Item=(Type=Control,Name="dyn_chain_ctrl_l_control"),Weight=0.75))
    Weight=@n6.Result
   )
   > &RigUnit_ParentConstraint(
    AdvancedSettings=(RotationOrderForFilter=XZY)
    Child=(Type=Null,Name="dyn_chain_3_front_r_null")
    Parents=((Item=(Type=Bone,Name="thigh_r"),Weight=0.9),(Item=(Type=Control,Name="dyn_chain_ctrl_r_control"),Weight=0.1))
    Weight=1
   )
   n7 = RigVMVariableNode() {variable=spr_thigh_r_up_alpha}
   n8 = &RigVMFunction_MathDoubleSub(A=1, B=@n7.Value)
   > &RigUnit_ParentConstraint(
    AdvancedSettings=(RotationOrderForFilter=XZY)
    Child=(Type=Null,Name="dyn_chain_5_front_r_null")
    Parents=((Item=(Type=Bone,Name="thigh_r"),Weight=0.75),(Item=(Type=Control,Name="dyn_chain_ctrl_r_control"),Weight=0.25))
    Weight=@n8.Result
   )
   > &RigUnit_ParentConstraint(
    AdvancedSettings=(RotationOrderForFilter=XZY)
    Child=(Type=Null,Name="dyn_chain_2_back_r_null")
    Parents=((Item=(Type=Bone,Name="thigh_r"),Weight=0.3),(Item=(Type=Control,Name="dyn_chain_ctrl_r_control"),Weight=0.3),(Item=(Type=Control,Name="dyn_chain_base_back_r_control"),Weight=0.4))
    Weight=@n8.Result
   )
   > &RigUnit_ParentConstraint(
    AdvancedSettings=(RotationOrderForFilter=XZY)
    Child=(Type=Null,Name="dyn_chain_3_back_r_null")
    Parents=((Item=(Type=Bone,Name="thigh_r"),Weight=0.15),(Item=(Type=Control,Name="dyn_chain_ctrl_r_control"),Weight=0.75),(Item=(Type=Control,Name="dyn_chain_end_back_r_null"),Weight=0.1))
    Weight=@n8.Result
   )
   > &RigUnit_ParentConstraint(
    AdvancedSettings=(RotationOrderForFilter=XZY)
    Child=(Type=Null,Name="dyn_chain_5_front_r_null")
    Parents=((Item=(Type=Bone,Name="thigh_r"),Weight=0.5),(Item=(Type=Control,Name="dyn_chain_ctrl_r_control"),Weight=0.5))
    Weight=@n7.Value
   )
   > &RigUnit_ParentConstraint(
    AdvancedSettings=(RotationOrderForFilter=XZY)
    Child=(Type=Control,Name="dyn_chain_2_back_r_control")
    Parents=((Item=(Type=Control,Name="dyn_chain_base_back_r_control"),Weight=0.5),(Item=(Type=Control,Name="dyn_chain_ctrl_r_control"),Weight=0.5))
    Weight=@n7.Value
   )
   > &RigUnit_ParentConstraint(
    AdvancedSettings=(RotationOrderForFilter=XZY)
    Child=(Type=Control,Name="dyn_chain_3_back_r_control")
    Parents=((Item=(Type=Control,Name="dyn_chain_base_back_r_control"),Weight=0.25),(Item=(Type=Control,Name="dyn_chain_ctrl_r_control"),Weight=0.75))
    Weight=@n7.Value
   )
   &RigUnit_ParentConstraint(
    AdvancedSettings=(RotationOrderForFilter=XZY)
    Child=(Type=Null,Name="dyn_chain_2_back_l_null")
    Parents=((Item=(Type=Bone,Name="pelvis"),Weight=0.75),(Item=(Type=Control,Name="dyn_chain_ctrl_l_control"),Weight=0.25))
    ExecutePin=@SequenceAggregate_1.B
    Weight=@n5.Value
   )
   > &RigUnit_ParentConstraint(
    AdvancedSettings=(RotationOrderForFilter=XZY)
    Child=(Type=Null,Name="dyn_chain_back_3_l_null")
    Parents=((Item=(Type=Bone,Name="pelvis"),Weight=0.25),(Item=(Type=Control,Name="dyn_chain_ctrl_l_control"),Weight=0.75))
    Weight=@n5.Value
   )
   SequenceAggregate_2 = RigVMAggregateNode(ExecuteContext=@SequenceAggregate.G)
   n9 = RigVMVariableNode() {variable=crouch_notmove}
   RigVMFunction_MathQuaternionFromEuler = &RigVMFunction_MathQuaternionFromEuler(
    RotationOrder=XYZ
    Euler=RigVMDispatchNode(False=(X=130,Y=-50,Z=-20), True=(X=90,Y=-40,Z=-10), Condition=@n9.Value) {template="DISPATCH_RigVMDispatch_If(in Condition,in True,in False,out Result)"}.Result
   )
   RigUnit_SetRelativeTransformForItem = /Script/ControlRig.RigUnit_SetRelativeTransformForItem(
    Child=(Type=Bone,Name="dyn_L_frontCoat_base")
    Parent=(Type=Bone,Name="dyn_L_frontCoat_base")
    Value=(Rotation=(X=0,Y=0,Z=0,W=1),Translation=(X=0,Y=0,Z=0),Scale3D=(X=1,Y=1,Z=1))
    bParentInitial=False
    bPropagateToChildren=True
    Value.Rotation=@RigVMFunction_MathQuaternionFromEuler.Result
    ExecutePin=@SequenceAggregate_2.A
    Weight=RigVMVariableNode() {variable=spr_thigh_l_up_alpha}.Value
   ) {
    is=struct RigVMUnitNode
    template="Set Relative Transform::Execute(in Child,in Parent,in bParentInitial,in Value,in Weight,in bPropagateToChildren)"
   }
   > &RigUnit_SetRelativeTransformForItem(
    Child=(Type=Bone,Name="dyn_R_frontCoat_base")
    Parent=(Type=Bone,Name="dyn_R_frontCoat_base")
    Value=(Rotation=(X=0,Y=0,Z=0,W=1),Translation=(X=0,Y=0,Z=0),Scale3D=(X=1,Y=1,Z=1))
    bParentInitial=False
    bPropagateToChildren=True
    Value.Rotation=&RigVMFunction_MathQuaternionFromEuler(
     RotationOrder=XYZ
     Euler=RigVMDispatchNode(False=(X=110,Y=-50,Z=-20), True=(X=100,Y=-50,Z=-10), Condition=@n9.Value) {template="DISPATCH_RigVMDispatch_If(in Condition,in True,in False,out Result)"}.Result
    ).Result
    Weight=RigVMVariableNode() {variable=spr_thigh_r_up_alpha}.Value
   ) {template="Set Relative Transform::Execute(in Child,in Parent,in bParentInitial,in Value,in Weight,in bPropagateToChildren)"}
   > n10 = &RigUnit_SphericalPoseReader(
    ActiveRegionScaleFactors=(PositiveWidth=3,NegativeWidth=3,PositiveHeight=2,NegativeHeight=0.75)
    ActiveRegionSize=0.1
    Debug=(bDrawDebug=false,bDrawLocalAxes=false,DebugScale=25,DebugSegments=20,DebugThickness=0.25)
    DriverAxis=(X=1,Y=0,Z=0)
    DriverItem=(Type=Bone,Name="upperarm_l")
    FalloffRegionScaleFactors=(PositiveWidth=1,NegativeWidth=1,PositiveHeight=1,NegativeHeight=1)
    FalloffSize=0.4
    OptionalParentItem=(Type=Bone)
    RotationOffset=(X=110,Y=-45,Z=0)
   )
   > &RigUnit_ModifyTransforms(
    ItemToModify=((Item=(Type=Bone,Name="dyn_coat_collar_up_parent_l"),Transform=(Rotation=(X=0,Y=0,Z=0,W=1),Translation=(X=0,Y=0,Z=0),Scale3D=(X=1,Y=1,Z=1))))
    ItemToModify.0.Transform.Rotation=&RigVMFunction_MathQuaternionFromEuler(Euler=(X=0,Y=0,Z=40)).Result
    Weight=@n10.OutputParam
   )
   > n11 = &RigUnit_SphericalPoseReader(
    ActiveRegionScaleFactors=(PositiveWidth=3,NegativeWidth=3,PositiveHeight=2,NegativeHeight=2)
    ActiveRegionSize=0.1
    Debug=(bDrawDebug=false,DebugScale=25,DebugSegments=20,DebugThickness=0.25)
    DriverAxis=(X=-1,Y=0,Z=0)
    DriverItem=(Type=Bone,Name="upperarm_r")
    FalloffRegionScaleFactors=(PositiveWidth=2,NegativeWidth=2,PositiveHeight=1,NegativeHeight=2)
    FalloffSize=0.4
    FlipWidthScaling=true
    OptionalParentItem=(Type=Bone)
    RotationOffset=(X=130,Y=225,Z=180)
   )
   > &RigUnit_ModifyTransforms(
    ItemToModify=((Item=(Type=Bone,Name="dyn_coat_collar_up_parent_r"),Transform=(Rotation=(X=0,Y=0,Z=0,W=1),Translation=(X=0,Y=0,Z=0),Scale3D=(X=1,Y=1,Z=1))))
    ItemToModify.0.Transform.Rotation=&RigVMFunction_MathQuaternionFromEuler(Euler=(X=0,Y=0,Z=110)).Result
    Weight=@n11.OutputParam
   )
   n12 = &RigUnit_SphericalPoseReader(
    ActiveRegionScaleFactors=(PositiveWidth=0.7,NegativeWidth=1,PositiveHeight=1.8,NegativeHeight=1.8)
    ActiveRegionSize=0.15
    Debug=(bDrawDebug=false,bDraw2D=false,bDrawLocalAxes=false,DebugScale=25,DebugSegments=20,DebugThickness=0.25)
    DriverAxis=(X=1,Y=0,Z=0)
    DriverItem=(Type=Bone,Name="head")
    FalloffRegionScaleFactors=(PositiveWidth=2.5,NegativeWidth=1,PositiveHeight=2,NegativeHeight=2)
    FalloffSize=0.3
    FlipHeightScaling=false
    FlipWidthScaling=false
    OptionalParentItem=(Type=Bone,Name="spine_05")
    RotationOffset=(X=-65,Y=-20,Z=-20)
    ExecutePin=@SequenceAggregate_2.B
   )
   n13 = RigVMVariableNode() {variable=roll_alpha}
   Multiply = &RigVMFunction_MathDoubleMul(B=40, A=@n13.Value)
   > &RigUnit_SetRelativeTransformForItem(
    Child=(Type=Bone,Name="dyn_collar_middle_base_l")
    Parent=(Type=Bone,Name="dyn_collar_middle_base_l")
    Value=(Rotation=(X=0,Y=0,Z=0,W=1),Translation=(X=0,Y=0,Z=0),Scale3D=(X=1,Y=1,Z=1))
    bParentInitial=False
    bPropagateToChildren=True
    Value.Rotation=&RigVMFunction_MathQuaternionFromEuler(Euler=(X=0,Y=-20,Z=-90), RotationOrder=XYZ, Euler.X=@Multiply.Result).Result
    Weight=@n12.OutputParam
   ) {template="Set Relative Transform::Execute(in Child,in Parent,in bParentInitial,in Value,in Weight,in bPropagateToChildren)"}
   > &RigUnit_SetRelativeTransformForItem(
    Child=(Type=Bone,Name="dyn_collar_parent_r")
    Parent=(Type=Bone,Name="dyn_collar_parent_r")
    Value=(Rotation=(X=0,Y=0,Z=0,W=1),Translation=(X=0,Y=0,Z=0),Scale3D=(X=1,Y=1,Z=1))
    bParentInitial=False
    bPropagateToChildren=True
    Value.Rotation=&RigVMFunction_MathQuaternionFromEuler(Euler=(X=40,Y=0,Z=-60), RotationOrder=XYZ, Euler.Z=&RigVMFunction_MathDoubleMul(B=-60, A=@n13.Value).Result).Result
    Weight=@n12.OutputParam
   ) {template="Set Relative Transform::Execute(in Child,in Parent,in bParentInitial,in Value,in Weight,in bPropagateToChildren)"}
   > n14 = &RigUnit_SphericalPoseReader(
    ActiveRegionScaleFactors=(PositiveWidth=3,NegativeWidth=3,PositiveHeight=1.5,NegativeHeight=0.75)
    ActiveRegionSize=0.1
    Debug=(bDrawDebug=false,bDrawLocalAxes=true,DebugScale=25,DebugSegments=20,DebugThickness=0.25)
    DriverAxis=(X=1,Y=0,Z=0)
    DriverItem=(Type=Bone,Name="upperarm_l")
    FalloffRegionScaleFactors=(PositiveWidth=1,NegativeWidth=1,PositiveHeight=1,NegativeHeight=1)
    FalloffSize=0.4
    OptionalParentItem=(Type=Bone)
    RotationOffset=(X=20,Y=40,Z=0)
   )
   > &RigUnit_SetRelativeTransformForItem(
    Child=(Type=Bone,Name="dyn_collar_parent_l")
    Parent=(Type=Bone,Name="dyn_collar_parent_l")
    Value=(Rotation=(X=0,Y=0,Z=0,W=1),Translation=(X=0,Y=0,Z=0),Scale3D=(X=1,Y=1,Z=1))
    bParentInitial=False
    bPropagateToChildren=True
    Value.Rotation=&RigVMFunction_MathQuaternionFromEuler(Euler=(X=0,Y=0,Z=90), RotationOrder=XYZ).Result
    Weight=@n14.OutputParam
   ) {template="Set Relative Transform::Execute(in Child,in Parent,in bParentInitial,in Value,in Weight,in bPropagateToChildren)"}
   > n15 = &RigUnit_SphericalPoseReader(
    ActiveRegionScaleFactors=(PositiveWidth=3,NegativeWidth=3,PositiveHeight=1.5,NegativeHeight=0.75)
    ActiveRegionSize=0.1
    Debug=(bDrawDebug=false,DebugScale=25,DebugSegments=20,DebugThickness=0.25)
    DriverAxis=(X=-1,Y=0,Z=0)
    DriverItem=(Type=Bone,Name="upperarm_r")
    FalloffRegionScaleFactors=(PositiveWidth=1,NegativeWidth=1,PositiveHeight=1,NegativeHeight=1)
    FalloffSize=0.4
    FlipWidthScaling=true
    OptionalParentItem=(Type=Bone)
    RotationOffset=(X=-150,Y=25,Z=0)
   )
   > &RigUnit_SetRelativeTransformForItem(
    Child=(Type=Bone,Name="dyn_collar_parent_r")
    Parent=(Type=Bone,Name="dyn_collar_parent_r")
    Value=(Rotation=(X=0,Y=0,Z=0,W=1),Translation=(X=0,Y=0,Z=0),Scale3D=(X=1,Y=1,Z=1))
    bParentInitial=False
    bPropagateToChildren=True
    Value.Rotation=&RigVMFunction_MathQuaternionFromEuler(Euler=(X=0,Y=0,Z=20), RotationOrder=XYZ).Result
    Weight=@n15.OutputParam
   ) {template="Set Relative Transform::Execute(in Child,in Parent,in bParentInitial,in Value,in Weight,in bPropagateToChildren)"}
   &RigUnit_RotationConstraintLocalSpaceOffset(
    AdvancedSettings=(RotationOrderForFilter=XZY)
    Child=(Type=Bone,Name="dyn_strap_ctrl_1_base")
    Filter=(bX=True,bY=True,bZ=True)
    Parents=((Item=(Type=Bone,Name="thigh_r"),Weight=1))
    Weight=1
    ExecutePin=@SequenceAggregate_2.C
   )
   n16 = RigVMVariableNode() {variable=spr_thigh_r_up_alpha}
   Remap = &RigVMFunction_MathDoubleRemap(SourceMaximum=1, SourceMinimum=0.2, TargetMaximum=1, TargetMinimum=0, bClamp=False, Value=@n16.Value)
   > &RigUnit_SetRelativeTransformForItem(
    Child=(Type=Bone,Name="dyn_strap_ctrl_1")
    Parent=(Type=Bone,Name="dyn_strap_ctrl_1")
    Value=(Rotation=(X=0,Y=0,Z=0,W=1),Translation=(X=0,Y=0,Z=3),Scale3D=(X=1,Y=1,Z=1))
    bParentInitial=False
    bPropagateToChildren=True
    Weight=@Remap.Result
    Value.Rotation=&RigVMFunction_MathQuaternionFromEuler(
     Euler=(X=110,Y=-10,Z=15)
     Euler.X=&RigVMFunction_MathDoubleRemap(SourceMaximum=1, SourceMinimum=0.4, TargetMaximum=8, TargetMinimum=110, bClamp=False, Value=@n16.Value).Result
    ).Result
   ) {template="Set Relative Transform::Execute(in Child,in Parent,in bParentInitial,in Value,in Weight,in bPropagateToChildren)"}
   > &RigUnit_SetRelativeTransformForItem(
    Child=(Type=Bone,Name="dyn_strap_ctrl_2")
    Parent=(Type=Bone,Name="dyn_strap_ctrl_2")
    Value=(Rotation=(X=0,Y=0,Z=0,W=1),Translation=(X=0,Y=0,Z=0),Scale3D=(X=1,Y=1,Z=1))
    bParentInitial=False
    bPropagateToChildren=True
    Value.Rotation=&RigVMFunction_MathQuaternionFromEuler(Euler=(X=0,Y=20,Z=0)).Result
    Weight=@n16.Value
   ) {template="Set Relative Transform::Execute(in Child,in Parent,in bParentInitial,in Value,in Weight,in bPropagateToChildren)"}
   > &RigUnit_SetRelativeTransformForItem(
    Child=(Type=Bone,Name="dyn_strap_ctrl_2")
    Parent=(Type=Bone,Name="dyn_strap_ctrl_2")
    Value=(Rotation=(X=0,Y=0,Z=0,W=1),Translation=(X=0,Y=0,Z=0),Scale3D=(X=1,Y=1,Z=1))
    bParentInitial=False
    bPropagateToChildren=True
    Value.Rotation=&RigVMFunction_MathQuaternionFromEuler(Euler=(X=20,Y=-8,Z=0)).Result
    Weight=RigVMVariableNode() {variable=frontend_alpha}.Value
   ) {template="Set Relative Transform::Execute(in Child,in Parent,in bParentInitial,in Value,in Weight,in bPropagateToChildren)"}
   > &RigUnit_RotationConstraintLocalSpaceOffset(
    AdvancedSettings=(RotationOrderForFilter=XZY)
    Child=(Type=Bone,Name="kneepad")
    Filter=(bX=True,bY=True,bZ=True)
    Parents=((Item=(Type=Bone,Name="thigh_r"),Weight=0.7),(Item=(Type=Bone,Name="calf_r"),Weight=0.3))
    Weight=1
   )
   > n17 = &RigUnit_SphericalPoseReader(
    ActiveRegionScaleFactors=(PositiveWidth=0.5,NegativeWidth=2,PositiveHeight=4,NegativeHeight=3.5)
    ActiveRegionSize=0.1
    Debug=(bDrawDebug=false,DebugScale=25,DebugSegments=20,DebugThickness=0.25)
    DriverAxis=(X=-1,Y=0,Z=0)
    DriverItem=(Type=Bone,Name="upperarm_r")
    FalloffRegionScaleFactors=(PositiveWidth=0.5,NegativeWidth=0.2,PositiveHeight=4,NegativeHeight=2)
    FalloffSize=0.4
    FlipWidthScaling=true
    OptionalParentItem=(Type=Bone)
    RotationOffset=(X=-80,Y=35,Z=-20)
   )
   > &RigUnit_SetRelativeTransformForItem(
    Child=(Type=Bone,Name="dyn_tie_base")
    Parent=(Type=Bone,Name="dyn_tie_base")
    Value=(Rotation=(X=0,Y=0,Z=0,W=1),Translation=(X=0,Y=0,Z=0),Scale3D=(X=1,Y=1,Z=1))
    bParentInitial=false
    bPropagateToChildren=True
    Weight=&RigVMFunction_MathDoubleMul(B=@n17.OutputParam, A=RigVMVariableNode() {variable=roll_alpha}.Value).Result
    Value.Rotation=&RigVMFunction_MathQuaternionFromEuler(Euler=(X=0,Y=-20,Z=0)).Result
   ) {template="Set Relative Transform::Execute(in Child,in Parent,in bParentInitial,in Value,in Weight,in bPropagateToChildren)"}
   RigVMFunction_MathBoolToFloat = &RigVMFunction_MathBoolToFloat(Value=RigVMVariableNode() {variable=fly}.Value)
   &RigUnit_SetRelativeTransformForItem(
    Child=(Type=Bone,Name="dyn_L_frontCoat_base")
    Parent=(Type=Bone,Name="dyn_L_frontCoat_base")
    Value=(Rotation=(X=0,Y=0,Z=0,W=1),Translation=(X=0,Y=0,Z=0),Scale3D=(X=1,Y=1,Z=1))
    bParentInitial=False
    bPropagateToChildren=True
    Weight=@RigVMFunction_MathBoolToFloat.Result
    Value.Rotation=&RigVMFunction_MathQuaternionFromEuler(Euler=(X=0,Y=-20,Z=0), RotationOrder=XYZ).Result
    ExecutePin=@SequenceAggregate_2.D
   ) {template="Set Relative Transform::Execute(in Child,in Parent,in bParentInitial,in Value,in Weight,in bPropagateToChildren)"}
   > &RigUnit_SetRelativeTransformForItem(
    Child=(Type=Bone,Name="dyn_R_frontCoat_base")
    Parent=(Type=Bone,Name="dyn_R_frontCoat_base")
    Value=(Rotation=(X=0,Y=0,Z=0,W=1),Translation=(X=0,Y=0,Z=0),Scale3D=(X=1,Y=1,Z=1))
    bParentInitial=False
    bPropagateToChildren=True
    Weight=&RigVMFunction_MathBoolToFloat(Value=RigVMVariableNode() {variable=fly}.Value).Result
    Value.Rotation=&RigVMFunction_MathQuaternionFromEuler(Euler=(X=0,Y=-20,Z=0), RotationOrder=XYZ).Result
   ) {template="Set Relative Transform::Execute(in Child,in Parent,in bParentInitial,in Value,in Weight,in bPropagateToChildren)"}
   > &RigUnit_RotationConstraintLocalSpaceOffset(
    AdvancedSettings=(RotationOrderForFilter=XZY)
    Child=(Type=Bone,Name="dyn_keychain_1")
    Filter=(bX=true,bY=true,bZ=True)
    Parents=((Item=(Type=Bone,Name="thigh_l"),Weight=1))
    Weight=1
   )
   > &RigUnit_RotationConstraintLocalSpaceOffset(
    AdvancedSettings=(RotationOrderForFilter=XZY)
    Child=(Type=Bone,Name="dyn_keychain_2")
    Filter=(bX=true,bY=true,bZ=True)
    Parents=((Item=(Type=Bone,Name="thigh_l"),Weight=1))
    Weight=1
   )
   > &RigUnit_RotationConstraintLocalSpaceOffset(
    AdvancedSettings=(RotationOrderForFilter=XZY)
    Child=(Type=Bone,Name="dyn_keychain_3")
    Filter=(bX=true,bY=true,bZ=True)
    Parents=((Item=(Type=Bone,Name="thigh_l"),Weight=1))
    Weight=1
   )
   RerouteNode_1 = RigVMDispatchNode(Values="((Type=Bone,Name=\"dyn_chain_base_front_l\"),((),Type=Bone,Name=\"dyn_chain_1_front_l\"),((),Type=Bone,Name=\"dyn_chain_2_front_l\"),((),Type=Bone,Name=\"dyn_chain_3_front_l\"),((),Type=Bone,Name=\"dyn_chain_4_front_l\"),((),Type=Bone,Name=\"dyn_chain_5_front_l\"),(() ...(truncated)") {template="DISPATCH_RigVMDispatch_ArrayMake(in Values,out Array)"}
   RerouteNode_13 = RigVMDispatchNode(Values=((Type=Control,Name="dyn_chain_base_front_l_control"),((),Type=Control,Name="dyn_chain_3_front_l_control"),((),Type=Control,Name="dyn_chain_6_front_l_null"),((),Type=Control,Name="dyn_chain_8_front_l_null"))) {template="DISPATCH_RigVMDispatch_ArrayMake(in Values,out Array)"}
   SplineIK_1 = RigVMFunctionReferenceNode(
    Primary Axis=(X=1,Y=0,Z=0)
    Samples Per Segment=16
    Secondary Spline Direction=(X=0,Y=1,Z=0)
    Spline Mode=BSpline
    Stretch=true
    Up Axis=(X=0,Y=1,Z=0)
    debug=false
    Bones=@RerouteNode_1.Array
    Controls=@RerouteNode_13.Array
    ExecuteContext=@SequenceAggregate.E
   )
   RigUnit_ItemArray = &RigUnit_ItemArray(Items=((Type=Control,Name="dyn_chain_base_back_l_control"),(Type=Control,Name="dyn_chain_2_back_l_control"),(Type=Control,Name="dyn_chain_3_back_l_control"),(Type=Control,Name="dyn_chain_end_back_l_null")))
   n18 = &RigUnit_ItemArray(Items=((Type=Bone,Name="dyn_chain_base_back_l"),(Type=Bone,Name="dyn_chain_1_back_l"),(Type=Bone,Name="dyn_chain_2_back_l"),(Type=Bone,Name="dyn_chain_3_back_l"),(Type=Bone,Name="dyn_chain_4_back_l"),(Type=Bone,Name="dyn_chain_end_back_l")))
   > SplineIK = RigVMFunctionReferenceNode(
    Primary Axis=(X=1,Y=0,Z=0)
    Samples Per Segment=16
    Secondary Spline Direction=(X=0,Y=5,Z=0)
    Spline Mode=BSpline
    Stretch=true
    Up Axis=(X=0,Y=1,Z=0)
    debug=false
    Controls=@RigUnit_ItemArray.Items
    Bones=@n18.Items
   )
   RerouteNode_12 = RigVMDispatchNode(Values=((Type=Control,Name="dyn_chain_base_front_r_control"),((),Type=Control,Name="dyn_chain_3_front_r_control"),((),Type=Control,Name="dyn_chain_5_front_r_control"),((),Type=Control,Name="dyn_chain_7_front_r_null"))) {template="DISPATCH_RigVMDispatch_ArrayMake(in Values,out Array)"}
   RerouteNode_14 = RigVMDispatchNode(Values="((Type=Bone,Name=\"dyn_chain_base_front_r\"),((),Type=Bone,Name=\"dyn_chain_1_front_r\"),((),Type=Bone,Name=\"dyn_chain_2_front_r\"),((),Type=Bone,Name=\"dyn_chain_3_front_r\"),((),Type=Bone,Name=\"dyn_chain_4_front_r\"),((),Type=Bone,Name=\"dyn_chain_5_front_r\"),(() ...(truncated)") {template="DISPATCH_RigVMDispatch_ArrayMake(in Values,out Array)"}
   > SplineIK_1_1 = RigVMFunctionReferenceNode(
    Primary Axis=(X=1,Y=0,Z=0)
    Samples Per Segment=16
    Secondary Spline Direction=(X=0,Y=1,Z=0)
    Spline Mode=BSpline
    Stretch=true
    Up Axis=(X=0,Y=1,Z=0)
    debug=false
    Controls=@RerouteNode_12.Array
    Bones=@RerouteNode_14.Array
   )
   RerouteNode_15 = RigVMDispatchNode(Values="((Type=Bone,Name=\"dyn_chain_base_back_r\"),((),Type=Bone,Name=\"dyn_chain_1_back_r\"),((),Type=Bone,Name=\"dyn_chain_2_back_r\"),((),Type=Bone,Name=\"dyn_chain_3_back_r\"),((),Type=Bone,Name=\"dyn_chain_4_back_r\"),((),Type=Bone,Name=\"dyn_chain_5_back_r\"),((),Type= ...(truncated)") {template="DISPATCH_RigVMDispatch_ArrayMake(in Values,out Array)"}
   n19 = &RigUnit_ItemArray(Items=((Type=Control,Name="dyn_chain_base_back_r_control"),(Type=Control,Name="dyn_chain_2_back_r_control"),(Type=Control,Name="dyn_chain_3_back_r_control"),(Type=Control,Name="dyn_chain_end_back_r_null")))
   > SplineIK_1_1_1 = RigVMFunctionReferenceNode(
    Primary Axis=(X=1,Y=0,Z=0)
    Samples Per Segment=16
    Secondary Spline Direction=(X=0,Y=1,Z=0)
    Spline Mode=BSpline
    Stretch=true
    Up Axis=(X=0,Y=1,Z=0)
    debug=false
    Bones=@RerouteNode_15.Array
    Controls=@n19.Items
   )
   AlphaInterp = &RigVMFunction_AlphaInterp(
    InterpSpeedDecreasing=10
    InterpSpeedIncreasing=10
    Scale=1
    bClampResult=false
    Value=RigVMVariableNode() {variable=spr_thigh_l_up_alpha}.Value
   )
   n20 = RigVMVariableNode() {variable=SplineIK_chain_l_ctrls}
   &RigUnit_HierarchySetPoseItemArray(
    ElementType=All
    Weight=&RigVMFunction_MathDoubleSub(A=1, B=@AlphaInterp.Result).Result
    ItemsToSet=@n20.Value
    Pose=RigVMVariableNode() {variable=SplineIK_ctrls_l_bind_pose}.Value
    ExecutePin=@SequenceAggregate.F
   )
   > &RigUnit_HierarchySetPoseItemArray(ElementType=All, Weight=@AlphaInterp.Result, Pose=RigVMVariableNode() {variable=SplineIK_ctrl_thigh_l_up_pose}.Value, ItemsToSet=@n20.Value)
   n21 = RigVMVariableNode() {variable=spr_thigh_r_up_alpha}
   n22 = RigVMVariableNode() {variable=crouch_notmove}
   DISPATCH_RigVMDispatch_If = RigVMDispatchNode(True=0.5, False=&RigVMFunction_MathDoubleSub(A=1, B=@n21.Value).Result, Condition=@n22.Value) {template="DISPATCH_RigVMDispatch_If(in Condition,in True,in False,out Result)"}
   > &RigUnit_HierarchySetPoseItemArray(
    ElementType=All
    Weight=@DISPATCH_RigVMDispatch_If.Result
    Pose=RigVMVariableNode() {variable=SplineIK_ctrls_r_bind_pose}.Value
    ItemsToSet=@n2.Value
   )
   > &RigUnit_HierarchySetPoseItemArray(
    ElementType=All
    Weight=RigVMDispatchNode(True=0.5, Condition=@n22.Value, False=@n21.Value) {template="DISPATCH_RigVMDispatch_If(in Condition,in True,in False,out Result)"}.Result
    Pose=RigVMVariableNode() {variable=SplineIK_ctrl_thigh_r_up_pose}.Value
    ItemsToSet=@n2.Value
   )
   n23 = &RigUnit_SphericalPoseReader(
    ActiveRegionScaleFactors=(PositiveWidth=1,NegativeWidth=1,PositiveHeight=1,NegativeHeight=1)
    ActiveRegionSize=0.05
    Debug=(bDrawDebug=false,DebugScale=25,DebugSegments=20,DebugThickness=0.25)
    DriverAxis=(X=1,Y=0,Z=0)
    DriverItem=(Type=Bone,Name="thigh_r")
    ExecutePin=()
    FalloffRegionScaleFactors=(PositiveWidth=1,NegativeWidth=1,PositiveHeight=1,NegativeHeight=1)
    FalloffSize=0.09
    OptionalParentItem=(Type=Bone)
    RotationOffset=(X=15,Y=-80,Z=0)
   )
   > RigVMVariableNode(Value=@n23.OutputParam) {variable=spr_thigh_r_slight_back_out_apha}
  ]}
 RigVMModel Construction Event Graph {nodes [
   n1 = RigVMVariableNode() {variable=SplineIK_chain_r_ctrls}
   &RigUnit_HierarchySetPoseItemArray(ElementType=All, ExecutePin=(), Weight=1, Pose=RigVMVariableNode() {variable=SplineIK_ctrls_r_bind_pose}.Value, ItemsToSet=@n1.Value)
   > &RigUnit_ModifyTransforms(
    ItemToModify=((Item=(Type=Control,Name="dyn_chain_3_back_r_control"),Transform=(Rotation=(X=0,Y=0,Z=0,W=1),Translation=(X=0,Y=-0.5,Z=1),Scale3D=(X=1,Y=1,Z=1))))
    Weight=1
   )
   > &RigUnit_ModifyTransforms(
    ItemToModify=((Item=(Type=Control,Name="dyn_chain_ctrl_r_control"),Transform=(Rotation=(X=0,Y=0,Z=0,W=1),Translation=(X=1,Y=-0.1,Z=0),Scale3D=(X=1,Y=1,Z=1))))
    Weight=1
   )
   > &RigUnit_ModifyTransforms(
    ItemToModify=((Item=(Type=Control,Name="dyn_chain_3_front_r_control"),Transform=(Rotation=(X=0,Y=0,Z=0,W=1),Translation=(X=0,Y=0,Z=0),Scale3D=(X=1,Y=1,Z=1))))
    Weight=1
   )
   > &RigUnit_ModifyTransforms(
    ItemToModify=((Item=(Type=Control,Name="dyn_chain_5_front_r_control"),Transform=(Rotation=(X=0,Y=0,Z=0,W=1),Translation=(X=0,Y=0,Z=0),Scale3D=(X=1,Y=1,Z=1))))
    Weight=1
   )
   > RigVMVariableNode(Value=/Script/ControlRig.RigUnit_HierarchyGetPoseItemArray(ElementType=All, Initial=false, ItemsToGet=@n1.Value) {is=struct RigVMUnitNode}.Pose) {variable=SplineIK_ctrl_thigh_r_leg_slight_back_out_pose}
   n2 = &RigUnit_PrepareForExecution()
   SequenceAggregate = RigVMAggregateNode(ExecuteContext=@n2.ExecutePin)
   SequenceAggregate_1 = &RigVMFunction_Sequence(ExecuteContext=@SequenceAggregate.A)
   RerouteNode_13_1 = RigVMDispatchNode(Values="((Type=Control,Name=\"dyn_chain_base_front_l_control\"),((),Type=Control,Name=\"dyn_chain_3_front_l_control\"),((),Type=Control,Name=\"dyn_chain_6_front_l_null\"),((),Type=Control,Name=\"dyn_chain_8_front_l_null\"),((),Type=Control,Name=\"dyn_chain_base_back_l_cont ...(truncated)") {template="DISPATCH_RigVMDispatch_ArrayMake(in Values,out Array)"}
   RigVMVariableNode(Value=@RerouteNode_13_1.Array, ExecuteContext=@SequenceAggregate_1.A) {variable=SplineIK_chain_l_ctrls}
   RerouteNode_2 = RigVMDispatchNode(Values="((Type=Control,Name=\"dyn_chain_base_front_r_control\"),((),Type=Control,Name=\"dyn_chain_3_front_r_control\"),((),Type=Control,Name=\"dyn_chain_5_front_r_control\"),((),Type=Control,Name=\"dyn_chain_7_front_r_null\"),((),Type=Control,Name=\"dyn_chain_base_back_r_c ...(truncated)") {template="DISPATCH_RigVMDispatch_ArrayMake(in Values,out Array)"}
   > RigVMVariableNode(Value=@RerouteNode_2.Array) {variable=SplineIK_chain_r_ctrls}
   RigVMVariableNode(
    Value=&RigUnit_HierarchyGetPoseItemArray(ElementType=All, ItemsToGet=RigVMVariableNode() {variable=SplineIK_chain_l_ctrls}.Value).Pose
    ExecuteContext=@SequenceAggregate_1.B
   ) {variable=SplineIK_ctrls_l_bind_pose}
   > &RigUnit_ModifyTransforms(
    ItemToModify=((Item=(Type=Control,Name="dyn_chain_base_back_l_control"),Transform=(Rotation=(X=0,Y=0,Z=0,W=1),Translation=(X=0,Y=2,Z=-5),Scale3D=(X=1,Y=1,Z=1))))
    Weight=1
   )
   > &RigUnit_ModifyTransforms(
    ItemToModify=((Item=(Type=Control,Name="dyn_chain_2_back_l_control"),Transform=(Rotation=(X=0,Y=0,Z=0,W=1),Translation=(X=0,Y=2,Z=-5),Scale3D=(X=1,Y=1,Z=1))))
    Weight=1
   )
   > &RigUnit_ModifyTransforms(
    ItemToModify=((Item=(Type=Control,Name="dyn_chain_3_back_l_control"),Transform=(Rotation=(X=0,Y=0,Z=0,W=1),Translation=(X=0,Y=2,Z=-5),Scale3D=(X=1,Y=1,Z=1))))
    Weight=1
   )
   > &RigUnit_ModifyTransforms(
    ItemToModify=((Item=(Type=Control,Name="dyn_chain_base_back_r_control"),Transform=(Rotation=(X=0,Y=0,Z=0,W=1),Translation=(X=0,Y=5,Z=0),Scale3D=(X=1,Y=1,Z=1))))
    Weight=1
   )
   > &RigUnit_ModifyTransforms(
    ItemToModify=((Item=(Type=Control,Name="dyn_chain_2_back_r_control"),Transform=(Rotation=(X=0,Y=0,Z=0,W=1),Translation=(X=0,Y=5,Z=0),Scale3D=(X=1,Y=1,Z=1))))
    Weight=1
   )
   > RigVMVariableNode(Value=&RigUnit_HierarchyGetPoseItemArray(ElementType=All, ItemsToGet=RigVMVariableNode() {variable=SplineIK_chain_r_ctrls}.Value).Pose) {variable=SplineIK_ctrls_r_bind_pose}
   SequenceAggregate_2 = &RigVMFunction_Sequence(ExecuteContext=@SequenceAggregate.C)
   &RigUnit_ModifyTransforms(
    ItemToModify=((Item=(Type=Control,Name="dyn_chain_base_front_r_control"),Transform=(Rotation=(X=0,Y=0,Z=0,W=1),Translation=(X=-3,Y=0,Z=-3),Scale3D=(X=1,Y=1,Z=1))))
    Weight=1
    ExecutePin=@SequenceAggregate_2.A
   )
   > &RigUnit_ModifyTransforms(
    ItemToModify=((Item=(Type=Control,Name="dyn_chain_3_front_r_control"),Transform=(Rotation=(X=0,Y=0,Z=0,W=1),Translation=(X=0,Y=-7,Z=-7),Scale3D=(X=1,Y=1,Z=1))))
    Weight=1
   )
   > &RigUnit_ModifyTransforms(
    ItemToModify=((Item=(Type=Control,Name="dyn_chain_5_front_r_control"),Transform=(Rotation=(X=0,Y=0,Z=0,W=1),Translation=(X=-2,Y=-3,Z=-7),Scale3D=(X=1,Y=1,Z=1))))
    Weight=1
   )
   > &RigUnit_ModifyTransforms(
    ItemToModify=((Item=(Type=Control,Name="dyn_chain_7_front_r_null"),Transform=(Rotation=(X=0,Y=0,Z=0,W=1),Translation=(X=5,Y=2,Z=5),Scale3D=(X=1,Y=1,Z=1))))
    Weight=1
   )
   > &RigUnit_ModifyTransforms(
    ItemToModify=((Item=(Type=Control,Name="dyn_chain_3_back_r_control"),Transform=(Rotation=(X=0,Y=0,Z=0,W=1),Translation=(X=0,Y=0,Z=5),Scale3D=(X=1,Y=1,Z=1))))
    Weight=1
   )
   > &RigUnit_ModifyTransforms(
    ItemToModify=((Item=(Type=Control,Name="dyn_chain_end_back_r_null"),Transform=(Rotation=(X=0,Y=0,Z=0,W=1),Translation=(X=0,Y=2,Z=4),Scale3D=(X=1,Y=1,Z=1))))
    Weight=1
   )
   > RigVMVariableNode(Value=&RigUnit_HierarchyGetPoseItemArray(ElementType=All, ItemsToGet=RigVMVariableNode() {variable=SplineIK_chain_r_ctrls}.Value).Pose) {variable=SplineIK_ctrl_thigh_r_up_pose}
   &RigUnit_ModifyTransforms(
    ItemToModify=((Item=(Type=Control,Name="dyn_chain_3_front_l_control"),Transform=(Rotation=(X=0,Y=0,Z=0,W=1),Translation=(X=-10,Y=0,Z=3),Scale3D=(X=1,Y=1,Z=1))))
    Weight=1
    ExecutePin=@SequenceAggregate.B
   )
   > &RigUnit_ModifyTransforms(
    ItemToModify=((Item=(Type=Control,Name="dyn_chain_8_front_l_null"),Transform=(Rotation=(X=0,Y=0,Z=0,W=1),Translation=(X=-5,Y=0,Z=0),Scale3D=(X=1,Y=1,Z=1))))
    Weight=1
   )
   > &RigUnit_ModifyTransforms(
    ItemToModify=((Item=(Type=Control,Name="dyn_chain_base_back_l_control"),Transform=(Rotation=(X=0,Y=0,Z=0,W=1),Translation=(X=0,Y=5,Z=-1),Scale3D=(X=1,Y=1,Z=1))))
    Weight=1
   )
   > &RigUnit_ModifyTransforms(
    ItemToModify=((Item=(Type=Control,Name="dyn_chain_2_back_l_control"),Transform=(Rotation=(X=0,Y=0,Z=0,W=1),Translation=(X=-7,Y=0,Z=-3),Scale3D=(X=1,Y=1,Z=1))))
    Weight=1
   )
   > &RigUnit_ModifyTransforms(
    ItemToModify=((Item=(Type=Control,Name="dyn_chain_end_back_l_null"),Transform=(Rotation=(X=0,Y=0,Z=0,W=1),Translation=(X=-5,Y=0,Z=0),Scale3D=(X=1,Y=1,Z=1))))
    Weight=1
   )
   > RigVMVariableNode(Value=&RigUnit_HierarchyGetPoseItemArray(ElementType=All, ItemsToGet=RigVMVariableNode() {variable=SplineIK_chain_l_ctrls}.Value).Pose) {variable=SplineIK_ctrl_thigh_l_up_pose}
   RigVMVariableNode() {variable=SplineIK_back_chain_l_ctrls}
  ]}
 SequenceAggregate_2_1_ContainedGraph {nodes [
   n1 = RigVMFunctionEntryNode()
   SequenceAggregate_2 = &RigVMFunction_Sequence(ExecuteContext=@n1.ExecuteContext)
   SequenceAggregate_2_1 = &RigVMFunction_Sequence(ExecuteContext=@SequenceAggregate_2.B)
   SequenceAggregate_2_2 = &RigVMFunction_Sequence(ExecuteContext=@SequenceAggregate_2_1.B)
   RigVMFunctionReturnNode(A=@SequenceAggregate_2.A, B=@SequenceAggregate_2_1.A, C=@SequenceAggregate_2_2.A, D=@SequenceAggregate_2_2.B)
  ]}
]
hierarchy [
 root {loc=<0 0 0>, rot=<0 0 0 1>}
 pelvis {parent=root, loc=<0 -0.7071 93.1191>, rot=<0.0127 0.707 -0.0127 -0.707>}
 spine_01 {parent=pelvis, loc=<6.7494 -0.6388 0>, rot=<0 0 0.034 -0.9994>}
 spine_02 {parent=spine_01, loc=<7.2518 -0.035 0>, rot=<0 0 0.0238 -0.9997>}
 spine_03 {parent=spine_02, loc=<6.0104 0 0>, rot=<0 0 0.0149 -0.9999>}
 spine_04 {parent=spine_03, loc=<8.1024 0 0>, rot=<0 0 -0.0132 -0.9999>}
 spine_05 {parent=spine_04, loc=<10.0146 0 0>, rot=<0 0 -0.0015 -1>}
 attach_cape {parent=spine_05, loc=<9.017 -11.714 0>, rot=<0.0616 -0.7044 -0.0616 -0.7044>}
 neck_01 {parent=spine_05, loc=<11.3138 0.1529 0>, rot=<0 0 -0.2003 -0.9797>}
 neck_02 {parent=neck_01, loc=<4.5627 0 0>, rot=<0 0 0.0504 -0.9987>}
 head {parent=neck_02, loc=<5.3914 0 0>, rot=<0 0 0.115 -0.9934>}
 attach_backpack {parent=spine_05, loc=<-0.123 -13.473 0>, rot=<-0.0093 -0.707 0.0093 -0.707>}
 clavicle_l {parent=spine_05, loc=<4.3575 -0.1466 -2.3223>, rot=<0.0465 -0.7289 -0.009 -0.6829>}
 upperarm_l {parent=clavicle_l, loc=<15.5111 0 0>, rot=<0.0226 -0.3234 -0.0088 -0.9459>}
 lowerarm_l {parent=upperarm_l, loc=<26.6514 0 0>, rot=<0.0391 -0.0099 -0.2453 -0.9686>}
 hand_l {parent=lowerarm_l, loc=<22.3491 0 0>, rot=<-0.7081 -0.0507 0.0975 -0.6975>}
 thumb_01_l {parent=hand_l, loc=<1.3118 -0.3191 -2.459>, rot=<0.6546 -0.4042 0.0664 -0.6354>}
 thumb_02_l {parent=thumb_01_l, loc=<5.9336 -0.0448 0.0401>, rot=<-0.0037 0.0213 0.298 -0.9543>}
 thumb_03_l {parent=thumb_02_l, loc=<2.2223 0.3682 -0.0907>, rot=<0.0118 -0.013 0.0066 -0.9998>}
 index_metacarpal_l {parent=hand_l, loc=<1.6535 0.0183 -2.96>, rot=<0.0158 -0.0124 -0.0173 -0.9996>}
 index_01_l {parent=index_metacarpal_l, loc=<8.3008 0.0357 -0.0524>, rot=<-0.0044 0.0485 0.0924 -0.9945>}
 index_02_l {parent=index_01_l, loc=<3.3587 -0.3004 0.0452>, rot=<-0.002 0.0027 0.3225 -0.9466>}
 index_03_l {parent=index_02_l, loc=<2.5578 -0.123 -0.0252>, rot=<-0.002 0.0067 0.2558 -0.9667>}
 middle_metacarpal_l {parent=hand_l, loc=<2.1856 -0.1725 -0.5984>, rot=<0.006 0.0168 -0.0329 -0.9993>}
 middle_01_l {parent=middle_metacarpal_l, loc=<7.3275 -0.2447 0.0129>, rot=<-0.0208 0.0042 0.0858 -0.9961>}
 middle_02_l {parent=middle_01_l, loc=<4.0123 0.1401 0.1175>, rot=<-0.0218 0.0258 0.3512 -0.9357>}
 middle_03_l {parent=middle_02_l, loc=<2.6077 0.0534 0.1373>, rot=<0.0082 0.0188 0.1708 -0.9851>}
 ring_metacarpal_l {parent=hand_l, loc=<2.2074 -0.3625 1.3638>, rot=<-0.0284 0.0722 -0.0458 -0.9959>}
 ring_01_l {parent=ring_metacarpal_l, loc=<7.0295 -0.6197 -0.1446>, rot=<-0.0489 -0.0252 0.1892 -0.9804>}
 ring_02_l {parent=ring_01_l, loc=<3.6174 0.4647 -0.1027>, rot=<0 0 0.2806 -0.9598>}
 ring_03_l {parent=ring_02_l, loc=<2.4152 0.114 0.0199>, rot=<0 0 0.1189 -0.9929>}
 pinky_metacarpal_l {parent=hand_l, loc=<1.7616 -0.9753 3.3814>, rot=<-0.0397 0.0814 -0.069 -0.9935>}
 pinky_01_l {parent=pinky_metacarpal_l, loc=<7.1709 -0.4831 -0.1094>, rot=<-0.0654 -0.0035 0.2242 -0.9723>}
 pinky_02_l {parent=pinky_01_l, loc=<2.8274 0.0893 -0.1043>, rot=<0.0037 0.0011 0.2854 -0.9584>}
 pinky_03_l {parent=pinky_02_l, loc=<2.2046 0.0307 -0.0078>, rot=<-0.0002 0.0098 0.1445 -0.9895>}
 hand_attach_l {parent=hand_l, loc=<8.5 -3 0>, rot=<0.7071 0.7071 0 0>}
 dyn_L_braceletMaster {parent=hand_l, loc=<2.0401 -0.2688 0.2917>, rot=<0 0 0 -1>}
 dyn_L_braceletSecondary {parent=dyn_L_braceletMaster, loc=<1.2369 1.2228 0.1653>, rot=<0.5897 0.3448 -0.2848 -0.6725>}
 lowerarm_twist_02_l {parent=lowerarm_l, loc=<9.6199 0 0>, rot=<0 0 0 -1>}
 lowerarm_twist_01_l {parent=lowerarm_l, loc=<17.8549 0 0>, rot=<0 0 0 -1>}
 upperarm_twist_01_l {parent=upperarm_l, loc=<0.0041 0 0>, rot=<0 0 0 1>}
 upperarm_twist_02_l {parent=upperarm_l, loc=<13.7941 0 0>, rot=<0 0 0 1>}
 clavicle_r {parent=spine_05, loc=<4.3575 -0.1466 2.3223>, rot=<0.7289 0.0465 -0.6829 0.009>}
 upperarm_r {parent=clavicle_r, loc=<-15.5111 0 0>, rot=<0.0226 -0.3234 -0.0088 -0.9459>}
 lowerarm_r {parent=upperarm_r, loc=<-26.6514 0 0>, rot=<0.0391 -0.0099 -0.2453 -0.9686>}
 hand_r {parent=lowerarm_r, loc=<-22.3491 0 0>, rot=<-0.7081 -0.0507 0.0975 -0.6975>}
 thumb_01_r {parent=hand_r, loc=<-1.3118 0.3191 2.459>, rot=<0.6546 -0.4042 0.0664 -0.6354>}
 thumb_02_r {parent=thumb_01_r, loc=<-5.9336 0.0448 -0.0401>, rot=<-0.0037 0.0213 0.298 -0.9543>}
 thumb_03_r {parent=thumb_02_r, loc=<-2.2223 -0.3682 0.0907>, rot=<0.0118 -0.013 0.0066 -0.9998>}
 index_metacarpal_r {parent=hand_r, loc=<-1.6535 -0.0183 2.96>, rot=<0.0158 -0.0124 -0.0173 -0.9996>}
 index_01_r {parent=index_metacarpal_r, loc=<-8.3008 -0.0357 0.0524>, rot=<-0.0044 0.0485 0.0924 -0.9945>}
 index_02_r {parent=index_01_r, loc=<-3.3587 0.3004 -0.0452>, rot=<-0.002 0.0027 0.3225 -0.9466>}
 index_03_r {parent=index_02_r, loc=<-2.5578 0.123 0.0252>, rot=<-0.002 0.0067 0.2558 -0.9667>}
 middle_metacarpal_r {parent=hand_r, loc=<-2.1856 0.1725 0.5984>, rot=<0.006 0.0168 -0.0329 -0.9993>}
 middle_01_r {parent=middle_metacarpal_r, loc=<-7.3275 0.2447 -0.0129>, rot=<-0.0208 0.0042 0.0858 -0.9961>}
 middle_02_r {parent=middle_01_r, loc=<-4.0123 -0.1401 -0.1175>, rot=<-0.0218 0.0258 0.3512 -0.9357>}
 middle_03_r {parent=middle_02_r, loc=<-2.6077 -0.0534 -0.1373>, rot=<0.0082 0.0188 0.1708 -0.9851>}
 ring_metacarpal_r {parent=hand_r, loc=<-2.2074 0.3625 -1.3638>, rot=<-0.0284 0.0722 -0.0458 -0.9959>}
 ring_01_r {parent=ring_metacarpal_r, loc=<-7.0295 0.6197 0.1446>, rot=<-0.0489 -0.0252 0.1892 -0.9804>}
 ring_02_r {parent=ring_01_r, loc=<-3.6174 -0.4647 0.1027>, rot=<0 0 0.2806 -0.9598>}
 ring_03_r {parent=ring_02_r, loc=<-2.4152 -0.114 -0.0199>, rot=<0 0 0.1189 -0.9929>}
 pinky_metacarpal_r {parent=hand_r, loc=<-1.7616 0.9753 -3.3814>, rot=<-0.0397 0.0814 -0.069 -0.9935>}
 pinky_01_r {parent=pinky_metacarpal_r, loc=<-7.1709 0.4831 0.1094>, rot=<-0.0654 -0.0035 0.2242 -0.9723>}
 pinky_02_r {parent=pinky_01_r, loc=<-2.8274 -0.0893 0.1043>, rot=<0.0037 0.0011 0.2854 -0.9584>}
 pinky_03_r {parent=pinky_02_r, loc=<-2.2046 -0.0307 0.0078>, rot=<-0.0002 0.0098 0.1445 -0.9895>}
 dyn_R_braceletMaster {parent=hand_r, loc=<-2.0402 0.2689 -0.2917>, rot=<0 -1 0 0>}
 dyn_R_braceletSecondary {parent=dyn_R_braceletMaster, loc=<1.237 -1.2228 0.1654>, rot=<-0.2848 -0.6725 -0.5897 -0.3448>}
 lowerarm_twist_02_r {parent=lowerarm_r, loc=<-9.6199 0 0>, rot=<0 0 0 -1>}
 lowerarm_twist_01_r {parent=lowerarm_r, loc=<-17.8549 0 0>, rot=<0 0 0 -1>}
 upperarm_twist_01_r {parent=upperarm_r, loc=<-0.0041 0 0>, rot=<0 0 0 -1>}
 upperarm_twist_02_r {parent=upperarm_r, loc=<-13.7941 0 0>, rot=<0 0 0 -1>}
 dyn_tie_base {parent=spine_05, loc=<1.5336 11.5815 -0.0357>, rot=<-0.1066 -0.9943 0.0008 0.0044>}
 dyn_tie_1 {parent=dyn_tie_base, loc=<9.826 0 0>, rot=<0.0013 0.0064 0.0984 -0.9951>}
 dyn_tie_2 {parent=dyn_tie_1, loc=<9.1493 0 0>, rot=<0 -0.0003 0.033 -0.9995>}
 dyn_tie_3 {parent=dyn_tie_2, loc=<8.31 0 0>, rot=<0.707 -0.0119 0.0089 -0.7071>}
 dyn_tie_down_1 {parent=dyn_tie_base, loc=<8.8145 -0.8829 -0.0099>, rot=<0.0012 0.0062 0.0959 -0.9954>}
 dyn_tie_down_2 {parent=dyn_tie_down_1, loc=<8.2387 0 0>, rot=<0 0.0008 0.0476 -0.9989>}
 dyn_tie_down_3 {parent=dyn_tie_down_2, loc=<7.838 0 0>, rot=<0.7069 -0.0192 0.0082 -0.707>}
 dyn_coat_collar_up_parent_l {parent=spine_05, loc=<2.8984 10.3578 -5.9433>, rot=<-0.1954 -0.6939 0.037 -0.6921>}
 dyn_coat_collar_up_base_l {parent=dyn_coat_collar_up_parent_l, loc=<0 0 0>, rot=<0 0 0 1>}
 dyn_coat_collar_up_parent_r {parent=spine_05, loc=<2.8983 10.3578 5.9433>, rot=<0.6939 -0.1954 -0.6921 -0.037>}
 dyn_coat_collar_up_base_r {parent=dyn_coat_collar_up_parent_r, loc=<0 0 0>, rot=<0 0 0 1>}
 dyn_collar_parent_l {parent=spine_05, loc=<11.2742 4.7167 -8.2418>, rot=<-0.6395 -0.4 -0.0067 -0.6565>}
 dyn_collar_base_l {parent=dyn_collar_parent_l, loc=<0 0 0>, rot=<0 0 0 1>}
 dyn_collar_1_l {parent=dyn_collar_base_l, loc=<8 1 2>, rot=<0.0937 -0.0243 -0.0192 -0.9951>}
 dyn_strap_up_base {parent=dyn_collar_1_l, loc=<2.8973 0.8122 -5.688>, rot=<-0.0156 -0.5078 0.1256 -0.8521>}
 dyn_strap_up_1 {parent=dyn_strap_up_base, loc=<3.829 0 0>, rot=<-0.0647 -0.0012 0.0719 -0.9953>}
 dyn_strap_down_base {parent=dyn_collar_1_l, loc=<-1.8085 1.0705 -5.8827>, rot=<-0.0727 -0.5041 0.0466 -0.8593>}
 dyn_strap_down_1 {parent=dyn_strap_down_base, loc=<4.0219 -0.2648 0.0109>, rot=<-0.0164 -0.028 0.065 -0.9974>}
 dyn_collar_parent_r {parent=spine_05, loc=<11.2743 4.7167 8.2417>, rot=<-0.4 0.6395 0.6565 -0.0067>}
 dyn_collar_base_r {parent=dyn_collar_parent_r, loc=<0 0 0>, rot=<0 0 0 1>}
 dyn_collar_1_r {parent=dyn_collar_base_r, loc=<-8 1 -2>, rot=<0.0937 -0.0243 -0.0192 -0.9951>}
 dyn_collar_middle_parent_l {parent=spine_05, loc=<13.2938 -4.3902 0>, rot=<0 0 0.1516 -0.9884>}
 dyn_collar_middle_base_l {parent=dyn_collar_middle_parent_l, loc=<0 0 0>, rot=<0 0 0 1>}
 dyn_belt_keychain {parent=spine_02, loc=<-3.4846 10.6695 0.0095>, rot=<0.0107 -0.9999 -0.0001 -0.0008>}
 thigh_l {parent=pelvis, loc=<-2.5693 -0.0109 -7.7501>, rot=<0.0012 0.0572 0.0155 -0.9982>}
 calf_l {parent=thigh_l, loc=<-39.0484 0 0>, rot=<0.0006 -0.0161 -0.0391 -0.9991>}
 foot_l {parent=calf_l, loc=<-42.561 0.1077 0.4292>, rot=<0.0418 -0.0455 0.0343 -0.9975>}
 ball_l {parent=foot_l, loc=<-6.1599 13.8264 0.6806>, rot=<0.0001 -0.0001 -0.7063 -0.7079>}
 calf_twist_02_l {parent=calf_l, loc=<-18.8064 -0.0967 0.0397>, rot=<0.0237 0.0095 -0.0035 -0.9997>}
 calf_twist_01_l {parent=calf_l, loc=<-35.813 -0.0967 0.0397>, rot=<0.0237 0.0095 -0.0035 -0.9997>}
 thigh_twist_01_l {parent=thigh_l, loc=<-21.6303 0 0>, rot=<0.0013 0 0 -1>}
 thigh_r {parent=pelvis, loc=<-2.5693 -0.0109 7.7501>, rot=<0.0572 -0.0012 0.9982 0.0155>}
 calf_r {parent=thigh_r, loc=<39.0484 0 0>, rot=<0.0006 -0.0161 -0.0391 -0.9991>}
 foot_r {parent=calf_r, loc=<42.561 -0.1077 -0.4292>, rot=<0.0418 -0.0455 0.0343 -0.9975>}
 ball_r {parent=foot_r, loc=<6.1599 -13.8264 -0.6806>, rot=<0.0001 -0.0001 -0.7063 -0.7079>}
 calf_twist_02_r {parent=calf_r, loc=<18.8064 0.0967 -0.0397>, rot=<0.0237 0.0095 -0.0035 -0.9997>}
 calf_twist_01_r {parent=calf_r, loc=<35.813 0.0967 -0.0397>, rot=<0.0237 0.0095 -0.0035 -0.9997>}
 thigh_twist_01_r {parent=thigh_r, loc=<21.6303 0 0>, rot=<0.0013 0 0 -1>}
 kneepad {parent=thigh_r, loc=<38.8539 -3.6898 -0.0093>, rot=<-0.6655 0.0016 0.7464 0.0019>}
 dyn_keychain_1 {parent=pelvis, loc=<-0.0199 8.7547 -6.3678>, rot=<-0.0263 -0.9819 0.1833 -0.0404>}
 dyn_keychain_2 {parent=pelvis, loc=<1.7145 8.6721 -7.0453>, rot=<-0.0343 -0.9757 0.1996 -0.0832>}
 dyn_keychain_3 {parent=pelvis, loc=<3.4304 8.4504 -7.3153>, rot=<-0.0341 -0.9786 0.1686 -0.1126>}
 dyn_keychain_3_1 {parent=dyn_keychain_3, loc=<3.6816 0 0>, rot=<0 0 0 1>}
 dyn_chain_base_front_l {parent=pelvis, loc=<5.3227 7.662 -8.2995>, rot=<-0.0217 -0.9829 0.0025 -0.1826>}
 dyn_chain_1_front_l {parent=dyn_chain_base_front_l, loc=<5.2642 0 0>, rot=<-0.0003 0.0073 0.0555 -0.9984>}
 dyn_chain_2_front_l {parent=dyn_chain_1_front_l, loc=<3.8858 0 0>, rot=<0.0068 -0.0746 -0.0117 -0.9971>}
 dyn_chain_3_front_l {parent=dyn_chain_2_front_l, loc=<3.9059 0 0>, rot=<-0.0011 0.0098 0.0299 -0.9995>}
 dyn_chain_4_front_l {parent=dyn_chain_3_front_l, loc=<3.9222 0 0>, rot=<0.0005 -0.0027 0.0451 -0.999>}
 dyn_chain_5_front_l {parent=dyn_chain_4_front_l, loc=<3.6646 0 0>, rot=<-0.0336 0.0851 0.1461 -0.985>}
 dyn_chain_6_front_l {parent=dyn_chain_5_front_l, loc=<3.5908 0 0>, rot=<-0.1995 0.2203 0.2013 -0.9333>}
 dyn_chain_7_front_l {parent=dyn_chain_6_front_l, loc=<3.2893 0 0>, rot=<0.1508 -0.0669 0.2003 -0.9657>}
 dyn_chain_8_front_l {parent=dyn_chain_7_front_l, loc=<1.9212 0 0>, rot=<0 0 0 1>}
 dyn_chain_base_back_l {parent=pelvis, loc=<-5.455 -5.0976 -12.8195>, rot=<0.0261 -0.9945 0.0063 -0.1008>}
 dyn_chain_1_back_l {parent=dyn_chain_base_back_l, loc=<3.6712 0 0>, rot=<-0.0051 0.0697 -0.0155 -0.9974>}
 dyn_chain_2_back_l {parent=dyn_chain_1_back_l, loc=<3.733 0 0>, rot=<0.0009 0.0095 -0.1524 -0.9883>}
 dyn_chain_3_back_l {parent=dyn_chain_2_back_l, loc=<3.7404 0 0>, rot=<0.0173 0.0308 -0.2591 -0.9652>}
 dyn_chain_4_back_l {parent=dyn_chain_3_back_l, loc=<3.9159 0 0>, rot=<0.1616 0.0936 -0.2647 -0.9461>}
 dyn_chain_end_back_l {parent=dyn_chain_4_back_l, loc=<1.6867 0 0>, rot=<0 0 0 1>}
 dyn_chain_base_front_r {parent=pelvis, loc=<0.227 8.9338 7.9588>, rot=<-0.0041 -0.992 -0.004 0.126>}
 dyn_chain_1_front_r {parent=dyn_chain_base_front_r, loc=<3.3915 0 0>, rot=<0.0003 -0.0111 0.0033 -0.9999>}
 dyn_chain_2_front_r {parent=dyn_chain_1_front_r, loc=<3.2616 0 0>, rot=<0.0018 -0.0245 0.041 -0.9989>}
 dyn_chain_3_front_r {parent=dyn_chain_2_front_r, loc=<3.2433 0 0>, rot=<0.0071 -0.0481 0.0307 -0.9983>}
 dyn_chain_4_front_r {parent=dyn_chain_3_front_r, loc=<3.2754 0 0>, rot=<0.0107 -0.0388 0.0903 -0.9951>}
 dyn_chain_5_front_r {parent=dyn_chain_4_front_r, loc=<3.2726 0 0>, rot=<0.0673 -0.1284 0.1223 -0.9818>}
 dyn_chain_6_front_r {parent=dyn_chain_5_front_r, loc=<3.382 0 0>, rot=<0.1908 -0.1507 0.2828 -0.9279>}
 dyn_chain_7_front_r {parent=dyn_chain_6_front_r, loc=<1.6157 0 0>, rot=<0 0 0 1>}
 dyn_chain_base_back_r {parent=pelvis, loc=<-5.3487 -5.5371 12.6992>, rot=<-0.1636 -0.9495 0.036 0.2652>}
 dyn_chain_1_back_r {parent=dyn_chain_base_back_r, loc=<3.0367 0 0>, rot=<0.0291 0.0754 -0.0635 -0.9947>}
 dyn_chain_2_back_r {parent=dyn_chain_1_back_r, loc=<3.0476 0 0>, rot=<0.0181 0.0393 0.0007 -0.9991>}
 dyn_chain_3_back_r {parent=dyn_chain_2_back_r, loc=<2.9326 0 0>, rot=<0.0233 0.041 -0.0849 -0.9953>}
 dyn_chain_4_back_r {parent=dyn_chain_3_back_r, loc=<3.0257 0 0>, rot=<0.0465 0.0554 -0.0957 -0.9928>}
 dyn_chain_5_back_r {parent=dyn_chain_4_back_r, loc=<3.1698 0 0>, rot=<0.6647 0.4195 -0.1313 -0.6041>}
 dyn_chain_end_back_r {parent=dyn_chain_5_back_r, loc=<1.6304 0 0>, rot=<0 0 0 1>}
 dyn_chain_ctrl_1_r {parent=pelvis, loc=<-2.5114 1.6829 10.2956>, rot=<0.1968 0.3145 0.9286 0.0055>}
 dyn_chain_ctrl_2_r {parent=dyn_chain_ctrl_1_r, loc=<7.5952 0 0>, rot=<0 0 0 1>}
 dyn_chain_ctrl_end_r {parent=dyn_chain_ctrl_2_r, loc=<8.312 0 0>, rot=<0 0 0 1>}
 dyn_R_backCoat_1 {parent=pelvis, loc=<-1.8773 -9.3869 6.0781>, rot=<0.698 0.0532 -0.0598 -0.7116>}
 dyn_R_backCoat_2 {parent=dyn_R_backCoat_1, loc=<-15.591 0 0>, rot=<0.0007 -0.0071 -0.0129 -0.9999>}
 dyn_R_backCoat_3 {parent=dyn_R_backCoat_2, loc=<-16.4887 0.0001 0>, rot=<0 0 0 1>}
 dyn_R_sideCoat_1 {parent=pelvis, loc=<-1.6071 -2.7854 11.8178>, rot=<0.9503 0.0358 -0.1102 -0.2891>}
 dyn_R_sideCoat_2 {parent=dyn_R_sideCoat_1, loc=<-15.5909 0 -0.0001>, rot=<0.0007 -0.0071 -0.0129 -0.9999>}
 dyn_R_sideCoat_3 {parent=dyn_R_sideCoat_2, loc=<-16.4888 0 0>, rot=<0 0 0 -1>}
 dyn_L_backCoat_1 {parent=pelvis, loc=<-1.8773 -9.3869 -6.078>, rot=<0.0558 -0.7063 0.7033 -0.0572>}
 dyn_L_backCoat_2 {parent=dyn_L_backCoat_1, loc=<15.5909 0 0>, rot=<0.0007 -0.0071 -0.0129 -0.9999>}
 dyn_L_backCoat_3 {parent=dyn_L_backCoat_2, loc=<16.4887 0 0>, rot=<0 0 0 1>}
 dyn_L_sideCoat_1 {parent=pelvis, loc=<-1.6072 -2.785 -11.8176>, rot=<0.0358 -0.9503 0.2891 -0.1102>}
 dyn_L_sideCoat_2 {parent=dyn_L_sideCoat_1, loc=<15.5909 0 0>, rot=<0.0007 -0.0071 -0.0129 -0.9999>}
 dyn_L_sideCoat_3 {parent=dyn_L_sideCoat_2, loc=<16.4887 0 0>, rot=<0 0 0 1>}
 dyn_chain_ctrl_1_l {parent=pelvis, loc=<-0.107 1.1107 -10.7095>, rot=<0.0234 -0.9839 0.0105 -0.1767>}
 dyn_chain_ctrl_2_l {parent=dyn_chain_ctrl_1_l, loc=<8.468 0 0>, rot=<0 0 0 1>}
 dyn_chain_ctrl_end_l {parent=dyn_chain_ctrl_2_l, loc=<11.2752 0 0>, rot=<-0.0238 0.5703 -0.0342 -0.8204>}
 dyn_L_frontCoat_base {parent=pelvis, loc=<10.1467 4.4569 -11.9146>, rot=<0.0267 -0.9885 -0.0873 -0.1206>}
 dyn_L_frontCoat_1 {parent=dyn_L_frontCoat_base, loc=<10.2072 0 0>, rot=<0 0 0 1>}
 dyn_L_frontCoat_2 {parent=dyn_L_frontCoat_1, loc=<15.5909 0 0>, rot=<0.0007 -0.0071 -0.0129 -0.9999>}
 dyn_L_frontCoat_3 {parent=dyn_L_frontCoat_2, loc=<16.4887 0 0>, rot=<0 0 0 1>}
 dyn_R_frontCoat_base {parent=pelvis, loc=<10.147 4.4569 11.9146>, rot=<0.9885 0.0267 -0.1206 0.0873>}
 dyn_R_frontCoat_1 {parent=dyn_R_frontCoat_base, loc=<-10.2075 -0.0005 0.0002>, rot=<0 0 0 1>}
 dyn_R_frontCoat_2 {parent=dyn_R_frontCoat_1, loc=<-15.5909 0 0>, rot=<0.0007 -0.0071 -0.0129 -0.9999>}
 dyn_R_frontCoat_3 {parent=dyn_R_frontCoat_2, loc=<-16.4888 0 0>, rot=<0 0 0 1>}
 dyn_strap_ctrl_1_base {parent=pelvis, loc=<1.582 6.3812 8.708>, rot=<0.1324 0.458 0.8776 0.0504>}
 dyn_strap_ctrl_1 {parent=dyn_strap_ctrl_1_base, loc=<0 0 0>, rot=<0 0 0 1>}
 dyn_strap_ctrl_2 {parent=dyn_strap_ctrl_1, loc=<5.1206 0 0>, rot=<0 0 0 1>}
 dyn_chain_ctrl_l_null {type=Null, loc=<17.5658 -1.2443 74.5311>, rot=<0 0 0 1>}
 dyn_chain_ctrl_l_control {controlType=EULER_TRANSFORM, parent=dyn_chain_ctrl_l_null, type=Control, loc=<0 0 0>, rot=<0 0 0 1>}
 dyn_chain_8_front_l_null {controlType=EULER_TRANSFORM, parent=dyn_chain_ctrl_l_control, type=Control, loc=<0 0 0>, rot=<0 0 0 1>}
 dyn_chain_end_back_l_null {controlType=EULER_TRANSFORM, parent=dyn_chain_ctrl_l_control, type=Control, loc=<0 0 0>, rot=<0 0 0 1>}
 dyn_chain_base_front_l_null {type=Null, loc=<8.2995 7.1407 98.164>, rot=<0.0024 0.5661 0.0034 0.8243>}
 dyn_chain_base_front_l_control {controlType=EULER_TRANSFORM, parent=dyn_chain_base_front_l_null, type=Control, loc=<0 0 0>, rot=<0 0 0 1>}
 dyn_chain_3_front_l_null {type=Null, loc=<12.5249 6.477 85.8746>, rot=<0.0426 -0.6113 0.0549 -0.7883>}
 dyn_chain_3_front_l_control {controlType=EULER_TRANSFORM, parent=dyn_chain_3_front_l_null, type=Control, loc=<0 0 0>, rot=<0 0 0 1>}
 dyn_chain_6_front_l_null {type=Null, loc=<15.6565 3.3058 75.824>, rot=<-0.1213 0.2364 -0.4402 0.8577>}
 dyn_chain_6_front_l_null {controlType=EULER_TRANSFORM, parent=dyn_chain_6_front_l_null, type=Control, loc=<0 0 0>, rot=<0 0 0 1>}
 dyn_chain_base_back_l_null {type=Null, loc=<12.8195 -5.9969 87.8503>, rot=<-0.0279 0.6314 -0.0342 0.7742>}
 dyn_chain_base_back_l_control {controlType=EULER_TRANSFORM, parent=dyn_chain_base_back_l_null, type=Control, loc=<0 0 0>, rot=<0 0 0 1>}
 dyn_chain_back_3_l_null {type=Null, loc=<16.0878 -5.6127 77.3664>, rot=<-0.2032 -0.4992 -0.3176 -0.7801>}
 dyn_chain_3_back_l_control {controlType=EULER_TRANSFORM, parent=dyn_chain_back_3_l_null, type=Control, loc=<0 0 0>, rot=<0 0 0 1>}
 dyn_chain_base_front_r_null {type=Null, loc=<-7.9588 8.2291 93.0259>, rot=<-0.0109 0.7905 -0.0084 0.6123>}
 dyn_chain_base_front_r_control {controlType=EULER_TRANSFORM, parent=dyn_chain_base_front_r_null, type=Control, loc=<0 0 0>, rot=<0 0 0 1>}
 dyn_chain_3_front_r_null {type=Null, loc=<-10.7186 7.6488 83.5527>, rot=<0.0744 -0.8361 0.0482 -0.5414>}
 dyn_chain_3_front_r_control {controlType=EULER_TRANSFORM, parent=dyn_chain_3_front_r_null, type=Control, loc=<0 0 0>, rot=<0 0 0 1>}
 dyn_chain_5_front_r_null {type=Null, loc=<-13.512 5.9215 77.9247>, rot=<0.2763 -0.8831 0.1132 -0.3619>}
 dyn_chain_5_front_r_control {controlType=EULER_TRANSFORM, parent=dyn_chain_5_front_r_null, type=Control, loc=<0 0 0>, rot=<0 0 0 1>}
 dyn_chain_base_back_r_null {type=Null, loc=<-12.6992 -6.4322 87.9722>, rot=<0.1324 0.8604 0.0748 0.4863>}
 dyn_chain_base_back_r_control {controlType=EULER_TRANSFORM, parent=dyn_chain_base_back_r_null, type=Control, loc=<0 0 0>, rot=<0 0 0 1>}
 dyn_chain_3_back_r_null {type=Null, loc=<-15.9901 -3.0152 80.3717>, rot=<-0.2292 -0.7391 -0.1876 -0.6049>}
 dyn_chain_3_back_r_control {controlType=EULER_TRANSFORM, parent=dyn_chain_3_back_r_null, type=Control, loc=<0 0 0>, rot=<0 0 0 1>}
 dyn_chain_ctrl_r_null {type=Null, loc=<-16.056 2.4889 75.8085>, rot=<0.3437 -0.0957 0.8795 0.3149>}
 dyn_chain_ctrl_r_control {controlType=EULER_TRANSFORM, parent=dyn_chain_ctrl_r_null, type=Control, loc=<0 0 0>, rot=<0 0 0 1>}
 dyn_chain_7_front_r_null {controlType=EULER_TRANSFORM, parent=dyn_chain_ctrl_r_control, type=Control, loc=<0 0 0>, rot=<0 0 0 1>}
 dyn_chain_end_back_r_null {controlType=EULER_TRANSFORM, parent=dyn_chain_ctrl_r_control, type=Control, loc=<0 0 0>, rot=<0 0 0 1>}
 dyn_chain_2_back_l_null {type=Null, loc=<14.8057 -6.5337 80.7574>, rot=<0.0705 0.5641 0.1021 0.8163>}
 dyn_chain_2_back_l_control {controlType=EULER_TRANSFORM, parent=dyn_chain_2_back_l_null, type=Control, loc=<0 0 0>, rot=<0 0 0 1>}
 dyn_chain_2_back_r_null {type=Null, loc=<-14.8057 -6.5337 80.7574>, rot=<0.1021 0.8163 0.0705 0.5641>}
 dyn_chain_2_back_r_control {controlType=EULER_TRANSFORM, parent=dyn_chain_2_back_r_null, type=Control, loc=<0 0 0>, rot=<0 0 0 1>}
 thigh_l_spr {type=Curve, loc=<0 0 0>, rot=<0 0 0 1>}
 thigh_r_spr {type=Curve, loc=<0 0 0>, rot=<0 0 0 1>}
]
variables [
 SplineIK_back_chain_l_ctrls {
  type="TArray<FRigElementKey>"
  default=((Type=Control,Name="dyn_back_chain_01_l_ctrl"),(Type=Control,Name="dyn_back_chain_03_l_ctrl"),(Type=Control,Name="dyn_back_chain_05_l_ctrl"),(Type=Control,Name="dyn_back_chain_07_l_ctrl"),(Type=Control,Name="dyn_chain_end_l_ctrl"))
 }
 SplineIK_back_chain_r_ctrls {
  type="TArray<FRigElementKey>"
  default=((Type=Control,Name="dyn_back_chain_01_r_ctrl"),(Type=Control,Name="dyn_back_chain_03_r_ctrl"),(Type=Control,Name="dyn_back_chain_05_r_ctrl"),(Type=Control,Name="dyn_back_chain_07_r_ctrl"),(Type=Control,Name="dyn_chain_end_r_ctrl"))
 }
 SplineIK_chain_l_ctrls {
  type="TArray<FRigElementKey>"
  default="((Type=Control,Name=\"dyn_front_chain_01_l_ctrl\"),(Type=Control,Name=\"dyn_front_chain_03_l_ctrl\"),(Type=Control,Name=\"dyn_front_chain_05_l_ctrl\"),(Type=Control,Name=\"dyn_front_chain_07_l_ctrl\"),(Type=Control,Name=\"dyn_back_chain_01_l_ctrl\"),(Type=Control,Na ...(truncated)"
 }
 SplineIK_chain_r_ctrls {
  type="TArray<FRigElementKey>"
  default="((Type=Control,Name=\"dyn_front_chain_01_r_ctrl\"),(Type=Control,Name=\"dyn_front_chain_03_r_ctrl\"),(Type=Control,Name=\"dyn_front_chain_05_r_ctrl\"),(Type=Control,Name=\"dyn_front_chain_07_r_ctrl\"),(Type=Control,Name=\"dyn_back_chain_01_r_ctrl\"),(Type=Control,Na ...(truncated)"
 }
 SplineIK_ctrl_r_crouch_walk_left_pose {type=FRigPose, default=(Elements=,HierarchyTopologyVersion=-1,PoseHash=-1)}
 SplineIK_ctrl_thigh_l_up_pose {type=FRigPose, default=(Elements=,HierarchyTopologyVersion=-1,PoseHash=-1)}
 SplineIK_ctrl_thigh_r_leg_slight_back_out_pose {type=FRigPose, default=(Elements=,HierarchyTopologyVersion=-1,PoseHash=-1)}
 SplineIK_ctrl_thigh_r_up_pose {type=FRigPose, default=(Elements=,HierarchyTopologyVersion=-1,PoseHash=-1)}
 SplineIK_ctrls_l_bind_pose {type=FRigPose, default=(Elements=,HierarchyTopologyVersion=-1,PoseHash=-1)}
 SplineIK_ctrls_r_bind_pose {type=FRigPose, default=(Elements=,HierarchyTopologyVersion=-1,PoseHash=-1)}
 SplineIK_front_chain_l_ctrls {
  type="TArray<FRigElementKey>"
  default=((Type=Control,Name="dyn_front_chain_01_l_ctrl"),(Type=Control,Name="dyn_front_chain_03_l_ctrl"),(Type=Control,Name="dyn_front_chain_05_l_ctrl"),(Type=Control,Name="dyn_front_chain_07_l_ctrl"),(Type=Control,Name="dyn_chain_end_l_ctrl"))
 }
 SplineIK_front_chain_r_ctrls {
  type="TArray<FRigElementKey>"
  default=((Type=Control,Name="dyn_front_chain_01_r_ctrl"),(Type=Control,Name="dyn_front_chain_03_r_ctrl"),(Type=Control,Name="dyn_front_chain_05_r_ctrl"),(Type=Control,Name="dyn_front_chain_07_r_ctrl"),(Type=Control,Name="dyn_chain_end_r_ctrl"))
 }
 crouch_notmove {type=bool, default=False}
 fly {type=bool, default=False}
 frontend_alpha {type=double, default=0}
 roll_alpha {type=double, default=0}
 spr_thigh_l_up_alpha {type=double, default=0}
 spr_thigh_r_slight_back_out_apha {type=double, default=0}
 spr_thigh_r_up_alpha {type=double, default=0}
]
