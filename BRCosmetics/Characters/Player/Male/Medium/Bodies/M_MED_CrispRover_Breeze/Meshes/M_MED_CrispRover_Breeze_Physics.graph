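# Physics asset for the male medium CrispRover Breeze player body, set up for a dynamic coat around the pelvis and legs.
# Six three-segment capsule chains form front, side and back panels with limited angular joints, hung from the pelvis or, at the front, separate base bodies.
# Angular-free cross constraints tie neighboring panels at lower segments.
# Pelvis, thigh and calf capsules cover the lower body; two front chain capsules have no constraints.
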PhysicsAsset /BRCosmetics/Characters/Player/Male/Medium/Bodies/M_MED_CrispRover_Breeze/Meshes/M_MED_CrispRover_Breeze_Physics
bodies {
 calf_l [{shapeType=Capsule, length=35.654, radius=10.5695, center=<-28.8055 0.2209 3.6476>, rotation=<rot -3.3016 86.6217 -92.3669>}]
 calf_r [{shapeType=Capsule, length=36.257, radius=13.2321, center=<30.0752 -2.0008 -2.3976>, rotation=<rot 0.0868 87.4652 -85.155>}]
 dyn_L_backCoat_1 [{
   shapeType=Capsule
   length=9.9957
   radius=8.0043
   shapeName=dyn_L_backCoat_1_capsule
   center=<7.1292 0.4195 7.3761>
   rotation=<rot -7.7845 -1.1424 98.3756>
  }]
 dyn_L_backCoat_2 [{
   shapeType=Capsule
   length=14.1098
   radius=8
   shapeName=dyn_L_backCoat_2_capsule
   center=<7.2082 -0.0117 5.6291>
   rotation=<rot -5.2684 -0.6961 99.131>
  }]
 dyn_L_backCoat_3 [{
   shapeType=Capsule
   length=15.279
   radius=8
   shapeName=dyn_L_backCoat_3_capsule
   center=<7.1114 -0.2158 3.6425>
   rotation=<rot -8.0513 -1.182 98.3811>
  }]
 dyn_L_frontCoat_1 [{
   shapeType=Capsule
   length=12.027
   radius=5.973
   shapeName=dyn_L_frontCoat_1_capsule
   center=<7 0 6.8154>
   rotation=<rot -7.1216 4.916 91.7902>
  }]
 dyn_L_frontCoat_2 [{shapeType=Capsule, length=10.5, radius=7.5, shapeName=dyn_L_frontCoat_2_capsule, center=<7 0 6.0095>, rotation=<rot -9.4736 1.6848 90.2021>}]
 dyn_L_frontCoat_3 [{
   shapeType=Capsule
   length=10.0022
   radius=7.9978
   shapeName=dyn_L_frontCoat_3_capsule
   center=<7.3316 0.1206 4.272>
   rotation=<rot -10.2656 2.1075 88.6682>
  }]
 dyn_L_frontCoat_base [{shapeType=Capsule, length=2.4082, radius=1.3217, shapeName=dyn_L_frontCoat_base_capsule}]
 dyn_L_sideCoat_1 [{
   shapeType=Capsule
   length=10.246
   radius=7.754
   shapeName=dyn_L_sideCoat_1_capsule
   center=<8.0875 -0.4178 9.9757>
   rotation=<rot 0.0792 2.973 87.9309>
  }]
 dyn_L_sideCoat_2 [{shapeType=Capsule, length=14.1098, radius=8, shapeName=dyn_L_sideCoat_2_capsule, center=<7 0 7.1176>, rotation=<rot 0 0 90>}]
 dyn_L_sideCoat_3 [{shapeType=Capsule, length=14.3647, radius=8.9143, shapeName=dyn_L_sideCoat_3_capsule, center=<7 0 4>, rotation=<rot 0 0 90>}]
 dyn_R_backCoat_1 [{
   shapeType=Capsule
   length=10.6945
   radius=7.3055
   shapeName=dyn_R_backCoat_1_capsule
   center=<-7.3531 -0.0769 -6.4518>
   rotation=<rot -8.6337 -1.2016 95.3071>
  }]
 dyn_R_backCoat_2 [{
   shapeType=Capsule
   length=14.1098
   radius=8
   shapeName=dyn_R_backCoat_2_capsule
   center=<-6.989 0.1721 -4.1149>
   rotation=<rot -8.8352 -1.1355 97.3607>
  }]
 dyn_R_backCoat_3 [{
   shapeType=Capsule
   length=15.279
   radius=8
   shapeName=dyn_R_backCoat_3_capsule
   center=<-6.7377 0.324 -2.1118>
   rotation=<rot -8.2394 -1.0579 97.3492>
  }]
 dyn_R_frontCoat_1 [{
   shapeType=Capsule
   length=10.527
   radius=7.473
   shapeName=dyn_L_frontCoat_1_capsule
   center=<-6.8515 0.036 -6.0871>
   rotation=<rot -11.6977 4.3382 -88.0983>
  }]
 dyn_R_frontCoat_2 [{
   shapeType=Capsule
   length=10.5
   radius=7.5
   shapeName=dyn_L_frontCoat_2_capsule
   center=<-6.6996 0.0153 -4.2074>
   rotation=<rot -9.4736 1.6848 -89.7979>
  }]
 dyn_R_frontCoat_3 [{
   shapeType=Capsule
   length=10.2793
   radius=7.7207
   shapeName=dyn_R_frontCoat_3_capsule
   center=<-6.5931 0.0161 -1.8746>
   rotation=<rot -10.8451 3.1348 89.8353>
  }]
 dyn_R_frontCoat_base [{shapeType=Capsule, length=2.4082, radius=1.3217, shapeName=dyn_R_frontCoat_base_capsule}]
 dyn_R_sideCoat_1 [{
   shapeType=Capsule
   length=10.5
   radius=7.5
   shapeName=dyn_R_sideCoat_1_capsule
   center=<-7.4692 0 -7.2291>
   rotation=<rot -9.7554 0.0002 89.9999>
  }]
 dyn_R_sideCoat_2 [{shapeType=Capsule, length=14.1098, radius=8, shapeName=dyn_R_sideCoat_2_capsule, center=<-7 0 -4.5>, rotation=<rot -8.7132 0 90>}]
 dyn_R_sideCoat_3 [{
   shapeType=Capsule
   length=15.279
   radius=8
   shapeName=dyn_R_sideCoat_3_capsule
   center=<-7 0.7642 -6.473>
   rotation=<rot -8.9664 -0.5292 93.3936>
  }]
 dyn_chain_7_front_r [{
   shapeType=Capsule
   length=10.4376
   radius=9.5
   shapeName=dyn_chain_ctrl_end_r_capsule
   center=<-2.7863 -11.4685 -3.3541>
   rotation=<rot 2.208 159.4745 151.5297>
  }]
 dyn_chain_8_front_l [{
   shapeType=Capsule
   length=13.2193
   radius=8.6968
   shapeName=dyn_chain_ctrl_end_l_capsule
   center=<-2.0361 -11.7191 -1.9462>
   rotation=<rot 35.6698 20.9454 -49.8288>
  }]
 pelvis [{shapeType=Capsule, length=8.8375, radius=10.6297, center=<0.7943 0.5875 0.4506>, rotation=<rot 0 -81.5525 -1.5748>}]
 thigh_l [{shapeType=Capsule, length=32.2847, radius=8.3732, center=<-18.0477 -0.4039 0.6329>, rotation=<rot 0.0001 -86.3807 -88.0225>}]
 thigh_r [{shapeType=Capsule, length=28.8579, radius=8.5512, center=<19.6076 -0.4155 -1.9601>, rotation=<rot -0.0001 89.7561 -91.9287>}]
}
constraints [
 {bone1Name=thigh_l, bone2Name=pelvis, swing1LimitDegrees=45, swing2LimitDegrees=45, twistLimitDegrees=45}
 {bone1Name=calf_l, bone2Name=thigh_l, swing1LimitDegrees=45, swing2LimitDegrees=45, twistLimitDegrees=45}
 {bone1Name=thigh_r, bone2Name=pelvis, swing1LimitDegrees=45, swing2LimitDegrees=45, twistLimitDegrees=45}
 {bone1Name=calf_r, bone2Name=thigh_r, swing1LimitDegrees=45, swing2LimitDegrees=45, twistLimitDegrees=45}
 {bone1Name=dyn_L_backCoat_1, bone2Name=pelvis, swing1LimitDegrees=30, swing2LimitDegrees=45, twistLimitDegrees=2}
 {bone1Name=dyn_L_backCoat_2, bone2Name=dyn_L_backCoat_1, swing1LimitDegrees=30, swing2LimitDegrees=30, twistLimitDegrees=2}
 {bone1Name=dyn_L_backCoat_3, bone2Name=dyn_L_backCoat_2, swing1LimitDegrees=10, swing2LimitDegrees=10, twistLimitDegrees=10}
 {bone1Name=dyn_L_frontCoat_1, bone2Name=dyn_L_frontCoat_base, swing1LimitDegrees=15, swing2LimitDegrees=45, twistLimitDegrees=20}
 {bone1Name=dyn_L_frontCoat_2, bone2Name=dyn_L_frontCoat_1, swing1LimitDegrees=20, swing2LimitDegrees=30, twistLimitDegrees=10}
 {bone1Name=dyn_L_frontCoat_3, bone2Name=dyn_L_frontCoat_2, swing1LimitDegrees=10, swing2LimitDegrees=10, twistLimitDegrees=10}
 {bone1Name=dyn_L_sideCoat_1, bone2Name=pelvis, swing1LimitDegrees=5, swing2LimitDegrees=45, twistLimitDegrees=2}
 {bone1Name=dyn_L_sideCoat_2, bone2Name=dyn_L_sideCoat_1, swing1LimitDegrees=30, swing2LimitDegrees=30, twistLimitDegrees=2}
 {bone1Name=dyn_L_sideCoat_3, bone2Name=dyn_L_sideCoat_2, swing1LimitDegrees=10, swing2LimitDegrees=10, twistLimitDegrees=2}
 {bone1Name=dyn_R_backCoat_1, bone2Name=pelvis, swing1LimitDegrees=30, swing2LimitDegrees=45, twistLimitDegrees=2}
 {bone1Name=dyn_R_backCoat_2, bone2Name=dyn_R_backCoat_1, swing1LimitDegrees=30, swing2LimitDegrees=30, twistLimitDegrees=2}
 {bone1Name=dyn_R_backCoat_3, bone2Name=dyn_R_backCoat_2, swing1LimitDegrees=10, swing2LimitDegrees=10, twistLimitDegrees=10}
 {bone1Name=dyn_R_frontCoat_3, bone2Name=dyn_R_frontCoat_2, swing1LimitDegrees=10, swing2LimitDegrees=10, twistLimitDegrees=10}
 {bone1Name=dyn_R_sideCoat_1, bone2Name=pelvis, swing1LimitDegrees=2, swing2LimitDegrees=45, twistLimitDegrees=2}
 {bone1Name=dyn_R_sideCoat_2, bone2Name=dyn_R_sideCoat_1, swing1LimitDegrees=30, swing2LimitDegrees=30, twistLimitDegrees=2}
 {bone1Name=dyn_R_sideCoat_3, bone2Name=dyn_R_sideCoat_2, swing1LimitDegrees=15, swing2LimitDegrees=10, twistLimitDegrees=2}
 {
  bone1Name=dyn_R_backCoat_3
  bone2Name=dyn_L_backCoat_3
  swing1LimitDegrees=45
  swing1Motion=Free
  swing2LimitDegrees=45
  swing2Motion=Free
  twistLimitDegrees=45
  twistMotion=Free
 }
 {
  bone1Name=dyn_R_frontCoat_3
  bone2Name=dyn_R_sideCoat_3
  swing1LimitDegrees=45
  swing1Motion=Free
  swing2LimitDegrees=45
  swing2Motion=Free
  twistLimitDegrees=45
  twistMotion=Free
 }
 {
  bone1Name=dyn_L_frontCoat_3
  bone2Name=dyn_L_sideCoat_3
  swing1LimitDegrees=45
  swing1Motion=Free
  swing2LimitDegrees=45
  swing2Motion=Free
  twistLimitDegrees=45
  twistMotion=Free
 }
 {
  bone1Name=dyn_R_sideCoat_2
  bone2Name=dyn_R_backCoat_2
  swing1LimitDegrees=45
  swing1Motion=Free
  swing2LimitDegrees=45
  swing2Motion=Free
  twistLimitDegrees=45
  twistMotion=Free
 }
 {
  bone1Name=dyn_L_sideCoat_2
  bone2Name=dyn_L_backCoat_2
  swing1LimitDegrees=45
  swing1Motion=Free
  swing2LimitDegrees=45
  swing2Motion=Free
  twistLimitDegrees=45
  twistMotion=Free
 }
 {bone1Name=dyn_R_frontCoat_1, bone2Name=dyn_R_frontCoat_base, swing1LimitDegrees=15, swing2LimitDegrees=45, twistLimitDegrees=10}
 {bone1Name=dyn_R_frontCoat_2, bone2Name=dyn_R_frontCoat_1, swing1LimitDegrees=20, swing2LimitDegrees=20, twistLimitDegrees=10}
]
