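# Cosmetic body AnimBP for the CrispRover Breeze skin, adding secondary motion to coat, collars and bracelets.
# The copied parent pose passes through a Control Rig, AnimDynamics, and RigidBody sims for coat and collar.
# Slides blend in additive coat-bone offsets, and a specific frontend idle can constrain front coat bones to thighs.
# The EventGraph drives head twist, frontend alpha, reduced knee-slide coat simulation, and state-based gravity with skydive wind parameters.
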AnimBlueprint /BRCosmetics/Characters/Player/Male/Medium/Bodies/M_MED_CrispRover_Breeze/Meshes/M_MED_CrispRover_Breeze_AnimBP
target_skeleton=&/M_MED_CrispRover_Breeze_Skeleton
graphs [
 AnimGraph {nodes [
   n1 = Control Rig!(
    Source=Copy Pose From Mesh() {bUseAttachedParent=true}
    Alpha
    crouch_notmove:bool=Get bIsCrouching()
    fly:bool=Get bIsSkyDiving()
    roll_alpha:double=Get twist() -> (twist:float)
    frontend_alpha:double=Get frontendalpha() -> (frontendalpha:double)
   ) -> (Pose) {
    alphaBoolBlend {blendOption=Linear}
    controlRigAssetReference {blueprintRigClass=&/M_MED_CrispRover_Breeze_CtrlRig.M_MED_CrispRover_Breeze_CtrlRig_C}
   }
   n2 = Anim Dynamics - Bone: dyn_coat_collar_up_base_l(ComponentPose=Local To Component(LocalPose=@n1), Alpha=1, ExternalForce="0,0,0") {
    angularDampingOverride=1
    angularSpringConstant=20000
    bAngularSpring=true
    bOverrideAngularDamping=true
    bUseGravityOverride=true
    gravityScale=0.35
    simulationSpace=RootRelative
    alphaBoolBlend {blendOption=Linear}
    boundBone {boneName=dyn_coat_collar_up_base_l}
    componentLinearAccScale=<1 1 1>
    componentLinearVelScale=<1 1 1>
    gravityOverride {z=-980}
    physicsBodyDefinitions [{
      boundBone {boneName=dyn_coat_collar_up_base_l}
      boxExtents=<20 30 30>
      constraintSetup {angularLimitsMax=<0 0 20>, angularLimitsMin=<0 0 -5>, angularTarget=<1 0 0>}
      localJointOffset=<4 0 0>
     }]
   }
   n3 = Anim Dynamics - Bone: dyn_coat_collar_up_base_r(ComponentPose=@n2, Alpha=1, ExternalForce="0,0,0") {
    angularDampingOverride=1
    angularSpringConstant=20000
    bAngularSpring=true
    bOverrideAngularDamping=true
    bUseGravityOverride=true
    gravityScale=0.35
    simulationSpace=RootRelative
    alphaBoolBlend {blendOption=Linear}
    boundBone {boneName=dyn_coat_collar_up_base_r}
    componentLinearAccScale=<1 1 1>
    componentLinearVelScale=<1 1 1>
    gravityOverride {z=-980}
    physicsBodyDefinitions [{
      boundBone {boneName=dyn_coat_collar_up_base_r}
      boxExtents=<20 30 30>
      constraintSetup {angularLimitsMax=<0 0 5>, angularLimitsMin=<0 0 -20>, angularTarget=<1 0 0>}
      localJointOffset=<4 0 0>
     }]
   }
   n4 = Anim Dynamics - Bone: dyn_L_braceletSecondary(ComponentPose=@n3, Alpha=1, ExternalForce="0,0,0") {
    angularDampingOverride=1
    bLinearSpring=true
    bOverrideAngularDamping=true
    bUseGravityOverride=true
    gravityScale=0.35
    linearSpringConstant=100
    simulationSpace=RootRelative
    alphaBoolBlend {blendOption=Linear}
    boundBone {boneName=dyn_L_braceletSecondary}
    componentAppliedLinearAccClamp=<3000 3000 3000>
    componentLinearAccScale=<0.2 0.2 0.2>
    componentLinearVelScale=<0.2 0.2 0.2>
    gravityOverride {z=-980}
    physicsBodyDefinitions [{
      boundBone {boneName=dyn_L_braceletSecondary}
      boxExtents=<5 5 5>
      constraintSetup {angularLimitsMax=<10 0 10>, angularLimitsMin=<-10 -70 -10>, angularTarget=<1 0 0>, linearAxesMin=<0 0 1>}
      localJointOffset=<0.2 0 0>
     }]
    relativeSpaceBone {boneName=hand_l}
   }
   n5 = Anim Dynamics - Bone: dyn_R_braceletSecondary(ComponentPose=@n4, Alpha=1, ExternalForce="0,0,0") {
    angularDampingOverride=1
    bLinearSpring=true
    bOverrideAngularDamping=true
    bUseGravityOverride=true
    gravityScale=0.35
    linearSpringConstant=100
    simulationSpace=RootRelative
    alphaBoolBlend {blendOption=Linear}
    boundBone {boneName=dyn_R_braceletSecondary}
    componentAppliedLinearAccClamp=<3000 3000 3000>
    componentLinearAccScale=<0.2 0.2 0.2>
    componentLinearVelScale=<0.2 0.2 0.2>
    gravityOverride {z=980}
    physicsBodyDefinitions [{
      boundBone {boneName=dyn_R_braceletSecondary}
      boxExtents=<5 5 5>
      constraintSetup {angularLimitsMax=<10 0 10>, angularLimitsMin=<-10 -70 -10>, angularTarget=<1 0 0>, linearAxesMin=<0 0 1>}
      localJointOffset=<0.2 0 0>
     }]
    relativeSpaceBone {boneName=hand_r}
   }
   n6 = Anim Dynamics - Bone: dyn_collar_base_l(ComponentPose=@n5, Alpha=1, ExternalForce="0,0,0") {
    angularDampingOverride=1
    angularSpringConstant=20000
    bAngularSpring=true
    bOverrideAngularDamping=true
    bUseGravityOverride=true
    gravityScale=0.35
    simulationSpace=RootRelative
    alphaBoolBlend {blendOption=Linear}
    boundBone {boneName=dyn_collar_base_l}
    componentLinearAccScale=<1 1 1>
    componentLinearVelScale=<1 1 1>
    gravityOverride {z=-980}
    physicsBodyDefinitions [{
      boundBone {boneName=dyn_collar_base_l}
      boxExtents=<30 30 30>
      constraintSetup {angularLimitsMax=<5 5 20>, angularLimitsMin=<-5 -5 -20>, angularTarget=<1 0 0>}
      localJointOffset=<3 0 0>
     }]
   }
   n7 = Anim Dynamics - Bone: dyn_collar_base_r(ComponentPose=@n6, Alpha=1, ExternalForce="0,0,0") {
    angularDampingOverride=1
    angularSpringConstant=20000
    bAngularSpring=true
    bOverrideAngularDamping=true
    bUseGravityOverride=true
    gravityScale=0.35
    simulationSpace=RootRelative
    alphaBoolBlend {blendOption=Linear}
    boundBone {boneName=dyn_collar_base_r}
    componentLinearAccScale=<1 1 1>
    componentLinearVelScale=<1 1 1>
    gravityOverride {z=-980}
    physicsBodyDefinitions [{
      boundBone {boneName=dyn_collar_base_r}
      boxExtents=<30 30 30>
      constraintSetup {angularLimitsMax=<5 5 20>, angularLimitsMin=<-5 -5 -20>, angularTarget=<1 0 0>}
      localJointOffset=<3 0 0>
     }]
   }
   n8 = RigidBody!(ComponentPose=@n7, Alpha=Get coat_dyn_alpha() -> (coat_dyn_alpha:double), OverrideWorldGravity:Vector=Get Gravity Override(), ExternalForce) -> (Pose) {
    bClampLinearTranslationLimitToRefPose=true
    bOverrideWorldGravity=true
    overridePhysicsAsset=&/M_MED_CrispRover_Breeze_Physics
    simulationSpace=BaseBoneSpace
    alphaBoolBlend {blendOption=Linear}
    baseBoneRef {boneName=pelvis}
    componentAppliedLinearAccClamp=<1000 1000 1000>
    componentLinearAccScale=<0.4 0.4 0.4>
    componentLinearVelScale=<0.4 0.4 0.4>
    overrideWorldGravity {z=-980}
    simSpaceSettings {maxAngularAcceleration=150, maxAngularVelocity=1, maxLinearAcceleration=100, maxLinearVelocity=50, velocityScaleZ=0, worldAlpha=1}
   }
   n9 = RigidBody(ComponentPose=@n8, OverrideWorldGravity="0,0,-980") {
    bClampLinearTranslationLimitToRefPose=true
    bOverrideWorldGravity=true
    overridePhysicsAsset=&/M_MED_CrispRover_Breeze_Collar_Physics
    simulationSpace=BaseBoneSpace
    alphaBoolBlend {blendOption=Linear}
    baseBoneRef {boneName=spine_05}
    componentAppliedLinearAccClamp=<1000 1000 1000>
    componentLinearAccScale=<0.6 0.6 0.6>
    componentLinearVelScale=<0.6 0.6 0.6>
    overrideWorldGravity {z=-980}
    simSpaceSettings {maxAngularAcceleration=500, maxAngularVelocity=3, maxLinearAcceleration=400, maxLinearVelocity=250, velocityScaleZ=0, worldAlpha=1}
   }
   n10 = Get bIsBaseballSliding()
   n11 = "Transform (Modify) Bone - Bone: dyn_L_frontCoat_1"!(ComponentPose=@n9, bAlphaBoolEnabled:bool=@n10, Translation="0,0,0", Rotation="36.823,2.015,64.152", Scale="1,1,1") -> (Pose) {
    alphaInputType=Bool
    rotationMode=BMM_Additive
    rotationSpace=BCS_ParentBoneSpace
    translationMode=BMM_Additive
    translationSpace=BCS_ParentBoneSpace
    alphaBoolBlend {blendInTime=0.5, blendOption=Linear, blendOutTime=0.5}
    boneToModify {boneName=dyn_L_frontCoat_1}
    rotation=<rot 36.823 2.015 64.152>
   }
   n12 = "Transform (Modify) Bone - Bone: dyn_R_frontCoat_1"!(ComponentPose=@n11, bAlphaBoolEnabled:bool=@n10, Translation="0,0,0", Rotation="15.286,64.791,92.980", Scale="1,1,1") -> (Pose) {
    alphaInputType=Bool
    rotationMode=BMM_Additive
    rotationSpace=BCS_ParentBoneSpace
    translationMode=BMM_Additive
    translationSpace=BCS_ParentBoneSpace
    alphaBoolBlend {blendInTime=0.5, blendOption=Linear, blendOutTime=0.5}
    boneToModify {boneName=dyn_R_frontCoat_1}
    rotation=<rot 15.286 64.791 92.98>
   }
   n13 = "Transform (Modify) Bone - Bone: dyn_L_sideCoat_1"!(ComponentPose=@n12, bAlphaBoolEnabled:bool=@n10, Translation="0,0,0", Rotation="37.825,79.815,0.178", Scale="1,1,1") -> (Pose) {
    alphaInputType=Bool
    rotationMode=BMM_Additive
    rotationSpace=BCS_ParentBoneSpace
    translationMode=BMM_Additive
    translationSpace=BCS_ParentBoneSpace
    alphaBoolBlend {blendInTime=0.5, blendOption=Linear, blendOutTime=0.5}
    boneToModify {boneName=dyn_L_sideCoat_1}
    rotation=<rot 37.825 79.815 0.178>
   }
   n14 = "Transform (Modify) Bone - Bone: dyn_R_sideCoat_1"!(ComponentPose=@n13, bAlphaBoolEnabled:bool=@n10, Translation="0,0,0", Rotation="-53.881,94.116,-37.508", Scale="1,1,1") -> (Pose) {
    alphaInputType=Bool
    rotationMode=BMM_Additive
    rotationSpace=BCS_ParentBoneSpace
    translationMode=BMM_Additive
    translationSpace=BCS_ParentBoneSpace
    alphaBoolBlend {blendInTime=0.5, blendOption=Linear, blendOutTime=0.5}
    boneToModify {boneName=dyn_R_sideCoat_1}
    rotation=<rot -53.881 94.1159 -37.5079>
   }
   n15 = "Transform (Modify) Bone - Bone: dyn_L_backCoat_1"!(ComponentPose=@n14, bAlphaBoolEnabled:bool=@n10, Translation="0,0,0", Rotation="22.643,89.388,-6.813", Scale="1,1,1") -> (Pose) {
    alphaInputType=Bool
    rotationMode=BMM_Additive
    rotationSpace=BCS_ParentBoneSpace
    translationMode=BMM_Additive
    translationSpace=BCS_ParentBoneSpace
    alphaBoolBlend {blendInTime=0.5, blendOption=Linear, blendOutTime=0.5}
    boneToModify {boneName=dyn_L_backCoat_1}
    rotation=<rot 22.643 89.388 -6.813>
   }
   n16 = "Transform (Modify) Bone - Bone: dyn_R_backCoat_1"!(ComponentPose=@n15, bAlphaBoolEnabled:bool=@n10, Translation="0,0,0", Rotation="11.823,100.187,-17.248", Scale="1,1,1") -> (Pose) {
    alphaInputType=Bool
    rotationMode=BMM_Additive
    rotationSpace=BCS_ParentBoneSpace
    translationMode=BMM_Additive
    translationSpace=BCS_ParentBoneSpace
    alphaBoolBlend {blendInTime=0.5, blendOption=Linear, blendOutTime=0.5}
    boneToModify {boneName=dyn_R_backCoat_1}
    rotation=<rot 11.823 100.187 -17.248>
   }
   n17 = Get bIsKneeSliding()
   n18 = "Transform (Modify) Bone - Bone: dyn_L_frontCoat_1"(ComponentPose=@n16, bAlphaBoolEnabled=@n17, Translation="0,0,0", Rotation="61.911,-70.050,-59.744", Scale="1,1,1") {
    alphaInputType=Bool
    rotationMode=BMM_Additive
    rotationSpace=BCS_ParentBoneSpace
    translationMode=BMM_Additive
    translationSpace=BCS_ParentBoneSpace
    alphaBoolBlend {blendInTime=0.5, blendOption=Linear, blendOutTime=0.5}
    boneToModify {boneName=dyn_L_frontCoat_1}
    rotation=<rot 61.911 -70.05 -59.744>
   }
   n19 = "Transform (Modify) Bone - Bone: dyn_R_frontCoat_1"(ComponentPose=@n18, bAlphaBoolEnabled=@n17, Translation="0,0,0", Rotation="42.514,48.922,106.851", Scale="1,1,1") {
    alphaInputType=Bool
    rotationMode=BMM_Additive
    rotationSpace=BCS_ParentBoneSpace
    translationMode=BMM_Additive
    translationSpace=BCS_ParentBoneSpace
    alphaBoolBlend {blendInTime=0.5, blendOption=Linear, blendOutTime=0.5}
    boneToModify {boneName=dyn_R_frontCoat_1}
    rotation=<rot 42.514 48.922 106.851>
   }
   n20 = "Transform (Modify) Bone - Bone: dyn_L_sideCoat_1"(ComponentPose=@n19, bAlphaBoolEnabled=@n17, Translation="0,0,0", Rotation="20.978,91.755,52.234", Scale="1,1,1") {
    alphaInputType=Bool
    rotationMode=BMM_Additive
    rotationSpace=BCS_ParentBoneSpace
    translationMode=BMM_Additive
    translationSpace=BCS_ParentBoneSpace
    alphaBoolBlend {blendInTime=0.5, blendOption=Linear, blendOutTime=0.5}
    boneToModify {boneName=dyn_L_sideCoat_1}
    rotation=<rot 20.978 91.755 52.234>
   }
   n21 = "Transform (Modify) Bone - Bone: dyn_R_sideCoat_1"(ComponentPose=@n20, bAlphaBoolEnabled=@n17, Translation="0,0,0", Rotation="-81.459,54.778,19.784", Scale="1,1,1") {
    alphaInputType=Bool
    rotationMode=BMM_Additive
    rotationSpace=BCS_ParentBoneSpace
    translationMode=BMM_Additive
    translationSpace=BCS_ParentBoneSpace
    alphaBoolBlend {blendInTime=0.5, blendOption=Linear, blendOutTime=0.5}
    boneToModify {boneName=dyn_R_sideCoat_1}
    rotation=<rot -81.459 54.778 19.784>
   }
   n22 = "Transform (Modify) Bone - Bone: dyn_L_backCoat_1"(ComponentPose=@n21, bAlphaBoolEnabled=@n17, Translation="4.181,3.406,-0.179", Rotation="-19.980,91.769,23.121", Scale="1,1,1") {
    alphaInputType=Bool
    rotationMode=BMM_Additive
    rotationSpace=BCS_ParentBoneSpace
    translationMode=BMM_Additive
    translationSpace=BCS_ParentBoneSpace
    alphaBoolBlend {blendInTime=0.5, blendOption=Linear, blendOutTime=0.5}
    boneToModify {boneName=dyn_L_backCoat_1}
    rotation=<rot -19.98 91.769 23.121>
    translation=<4.181 3.406 -0.179>
   }
   n23 = "Transform (Modify) Bone - Bone: dyn_R_backCoat_1"(ComponentPose=@n22, bAlphaBoolEnabled=@n17, Translation="6.414,5.227,-0.264", Rotation="-62.640,105.050,-17.279", Scale="1,1,1") {
    alphaInputType=Bool
    rotationMode=BMM_Additive
    rotationSpace=BCS_ParentBoneSpace
    translationMode=BMM_Additive
    translationSpace=BCS_ParentBoneSpace
    alphaBoolBlend {blendInTime=0.5, blendOption=Linear, blendOutTime=0.5}
    boneToModify {boneName=dyn_R_backCoat_1}
    rotation=<rot -62.64 105.05 -17.279>
    translation=<6.4142 5.2272 -0.264>
   }
   n24 = Get engageDuringFrontendIdle() -> (engageDuringFrontendIdle:bool)
   n25 = "Constraint - dyn_L_frontCoat_1 (P)"!(ComponentPose=@n23, bAlphaBoolEnabled:bool=@n24, ConstraintWeights_0=1) -> (Pose) {
    alphaInputType=Bool
    alphaBoolBlend {blendInTime=0.05, blendOption=CircularOut, blendOutTime=0.05}
    boneToModify {boneName=dyn_L_frontCoat_1}
    constraintSetup [{targetBone {boneName=thigh_l}}]
    constraintWeights [1]
   }
   n26 = "Constraint - dyn_R_frontCoat_1 (P)"!(ComponentPose=@n25, bAlphaBoolEnabled:bool=@n24, ConstraintWeights_0=1) -> (Pose) {
    alphaInputType=Bool
    alphaBoolBlend {blendInTime=0.05, blendOption=CircularOut, blendOutTime=0.05}
    boneToModify {boneName=dyn_R_frontCoat_1}
    constraintSetup [{targetBone {boneName=thigh_r}}]
    constraintWeights [1]
   }
   Output Pose(Result=Component To Local(ComponentPose=@n26)) {layerGroup=None}
  ]}
 EventGraph {nodes [
   Event BlueprintUpdateAnimation()
   > n1 = Sequence()
   > Branch(Condition=Get bIsInFrontEnd())
   n2 = GetAnimInstance(self=GetOwnerSkeletalMeshComponent())
   > Is Valid(InputObject=@n2)
   n3 = Get bRigidBodyEnabled()
   n4 = Get bIsPlayingEmote()
   > Set engageDuringFrontendIdle(engageDuringFrontendIdle:bool=AND Boolean!(
     A=Montage_IsPlaying(self=@n2, Montage=/BRCosmetics/Animation/Game/MainPlayer/Menu/BR/CustomSkins/_s25/CrispRoverM/CrispRover_Male_Breeze_Idle_Montage)
     B=NOT Boolean(A=@n3)
     C:bool=NOT Boolean(A=@n4)
    ) -> (ReturnValue)) -> (Output_Get:bool)
   Set coat_dyn_alpha(execute=@n1.then_1, coat_dyn_alpha:double=SelectFloat(A=0.3, B=1, bPickA=Get bIsKneeSliding())) -> (Output_Get:double)
   > Set twist(twist:float=GetAlphaFromReferenceTransformDelta(
     SkeletalMeshComponent=GetOwnerSkeletalMeshComponent()
     BoneName=head
     BaseName=spine_01
     DeltaComponentType=QuaternionTwist
     InputRange=(X=5,Y=-20)
     TwistAxis="1,0,0"
    )) -> (Output_Get:float)
   > n5 = Gravity Override State Function DEPRECATED(CharacterPartAnimInstance=Self-Reference(), GameTimeInSeconds=GetGameTimeInSeconds(), GravityOverrideParams=Get Gravity Override Params())
   > Set Gravity Override(Gravity Override=@n5)
   > Set frontendalpha(frontendalpha:double=SelectFloat(A=1, B=0, bPickA=Get bIsInFrontEnd())) -> (Output_Get:double)
  ]}
]
variables [
 coat_dyn_alpha {type=number}
 engageDuringFrontendIdle {type=boolean}
 frontendalpha {type=number}
 Gravity Override {type=Vector}
 Gravity Override Params {
  type=GravityOverrideParamsStruct
  default {
   skydive_FloatingDown {gravityOverride=<0 0 -500>, windAmplitude=<10 -10 10>, windFrequency=<1 1 1>, windOffset=<0 0 0>}
   skydive_GoingUp {gravityOverride=<0 0 -500>, windAmplitude=<10 -10 10>, windFrequency=<1 1 1>}
   skydive_Parachuting {gravityOverride=<0 -800 -500>, windAmplitude=<0 10 10>, windFrequency=<1 1 1>}
  }
 }
 twist {type=number}
]
